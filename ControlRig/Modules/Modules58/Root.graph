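# Modular Control Rig Root module that builds the global, root and body offset controls for a character rig.
# It publishes shared colors, sizes, naming, mirror axis and control references as module metadata, then scales and colorizes its controls.
# It identifies known skeleton conventions by probing signature bone names, recording limb, spine and axis defaults in Root metadata.
# Optional socket generation places Spine and BipedPhysics drop points, with fallback searches when no pelvis is found.
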ControlRigBlueprint /ControlRig/Modules/Modules58/Root
graphs [
 CollapseNode_ContainedGraph {nodes [
   RigVMFunctionEntryNode()
   > Parent = RigVMFunctionReferenceNode(
    Child=/Script/ControlRig.RigUnit_ResolveConnector(Connector=(Type=Connector,Name="Root"), SkipSocket=true) {is=struct RigVMUnitNode}.Result
    New Parent=RigVMVariableNode() {variable=body offset control}.Value
   )
   > RigVMFunctionReturnNode()
  ]}
 CollapseNode_ContainedGraph {nodes [
   RigVMFunctionEntryNode()
   Concat = /Script/RigVM.RigVMFunction_NameConcat(A=RigVMVariableNode() {variable=Control Name Prefix}.Value, B=RigVMVariableNode() {variable=Global Control Name}.Value) {is=struct RigVMUnitNode, template="Concat::Execute(in A,in B,out Result)"}
   > Control Stack at Position_2 = RigVMFunctionReferenceNode(
    Add Bottom Null=False
    Add Null=False
    Add Secondary Control=False
    Control Suffix=_ctrl
    Global Position=(Rotation=(X=0,Y=0,Z=0,W=1),Translation=(X=0,Y=0,Z=0),Scale3D=(X=1,Y=1,Z=1))
    Null Suffix=_null
    Parent=(Type=None,Name="None")
    Secondary Settings="(DisplayName=\"None\",Shape=(bVisible=True,Name=\"Default\",Color=(R=1,G=0,B=0,A=1),Transform=(Rotation=(X=0,Y=0,Z=0,W=1),Translation=(X=0,Y=0,Z=0),Scale3D=(X=1,Y=1,Z=1))),Proxy=(bIsProxy=False,ShapeVisibility=BasedOnSelection),InitialSpace=LocalSpace,bUsePref ...(truncated)"
    Tag Controls=true
    Zero Controls=true
    Name=@Concat.Result
    Settings=RigVMVariableNode() {variable=Global Control Settings}.Value
   )
   > RigVMVariableNode(Value=@"Control Stack at Position_2".Control) {variable=global control}
   > Control Stack at Position_3 = RigVMFunctionReferenceNode(
    Add Bottom Null=False
    Add Null=False
    Add Secondary Control=False
    Control Suffix=_ctrl
    Global Position=(Rotation=(X=0,Y=0,Z=0,W=1),Translation=(X=0,Y=0,Z=0),Scale3D=(X=1,Y=1,Z=1))
    Null Suffix=_null
    Orient to World=true
    Secondary Settings="(DisplayName=\"None\",Shape=(bVisible=True,Name=\"Default\",Color=(R=1,G=0,B=0,A=1),Transform=(Rotation=(X=0,Y=0,Z=0,W=1),Translation=(X=0,Y=0,Z=0),Scale3D=(X=1,Y=1,Z=1))),Proxy=(bIsProxy=False,ShapeVisibility=BasedOnSelection),InitialSpace=LocalSpace,bUsePref ...(truncated)"
    Tag Controls=true
    Zero Controls=true
    Name=&RigVMFunction_NameConcat(A=RigVMVariableNode() {variable=Control Name Prefix}.Value, B=RigVMVariableNode() {variable=Root Control Name}.Value) {template="Concat::Execute(in A,in B,out Result)"}.Result
    Parent=@"Control Stack at Position_2".Control
    Settings=RigVMVariableNode() {variable=Root Ctrl Settings}.Value
   )
   > RigVMVariableNode(Value=@"Control Stack at Position_3".Control) {variable=Root Control}
   > Control Stack at Position_1_2 = RigVMFunctionReferenceNode(
    Add Bottom Null=False
    Add Null=False
    Add Secondary Control=False
    Control Suffix=_ctrl
    Global Position=(Rotation=(X=0,Y=0,Z=0,W=1),Translation=(X=0,Y=0,Z=0),Scale3D=(X=1,Y=1,Z=1))
    Null Suffix=_null
    Orient to World=true
    Secondary Settings="(DisplayName=\"None\",Shape=(bVisible=True,Name=\"Default\",Color=(R=1,G=0,B=0,A=1),Transform=(Rotation=(X=0,Y=0,Z=0,W=1),Translation=(X=0,Y=0,Z=0),Scale3D=(X=1,Y=1,Z=1))),Proxy=(bIsProxy=False,ShapeVisibility=BasedOnSelection),InitialSpace=LocalSpace,bUsePref ...(truncated)"
    Tag Controls=true
    Zero Controls=true
    Name=&RigVMFunction_NameConcat(A=RigVMVariableNode() {variable=Control Name Prefix}.Value, B=RigVMVariableNode() {variable=Body Offset Control Name}.Value) {template="Concat::Execute(in A,in B,out Result)"}.Result
    Parent=@"Control Stack at Position_2".Control
    Settings=RigVMVariableNode() {variable=Body Offset Control Settings}.Value
   )
   > RigVMVariableNode(Value=@"Control Stack at Position_1_2".Control) {variable=body offset control}
   > RigVMFunctionReturnNode()
  ]}
 CollapseNode_ContainedGraph {nodes [
   n1 = RigVMFunctionEntryNode()
   RigVMFunction_Sequence = &RigVMFunction_Sequence(ExecuteContext=@n1.Value)
   n2 = RigVMVariableNode() {variable=Root Module Settings}
   RigVMDispatchNode(Name=Color Center, NameSpace=Self, ExecuteContext=@RigVMFunction_Sequence.A, Value=@n2.Value.CenterColor_9_#1) {template="DISPATCH_RigDispatch_SetModuleMetadata(in Name,in NameSpace,in Value,out Success)"}
   > RigVMDispatchNode(Name=Color Left, NameSpace=Self, Value=@n2.Value.LeftColor_3_#2) {template="DISPATCH_RigDispatch_SetModuleMetadata(in Name,in NameSpace,in Value,out Success)"}
   > RigVMDispatchNode(Name=Color Right, NameSpace=Self, Value=@n2.Value.RightColor_5_#3) {template="DISPATCH_RigDispatch_SetModuleMetadata(in Name,in NameSpace,in Value,out Success)"}
   Get_Module_Metadata_6 = RigVMDispatchNode(Default=1, Name=Control Size, NameSpace=Self) {template="DISPATCH_RigDispatch_GetModuleMetadata(in Name,in NameSpace,in Default,out Value,out Found)"}
   > RigVMDispatchNode(
    Name=Control Size
    NameSpace=Self
    Value=RigVMDispatchNode(Condition=@Get_Module_Metadata_6.Found, True=@Get_Module_Metadata_6.Value, False=@n2.Value.ControlSize_12_#4) {template="DISPATCH_RigVMDispatch_If(in Condition,in True,in False,out Result)"}.Result
   ) {template="DISPATCH_RigDispatch_SetModuleMetadata(in Name,in NameSpace,in Value,out Success)"}
   > RigVMDispatchNode(Name=Prefix, NameSpace=Self, Value=@n2.Value.Prefix_20_#5) {template="DISPATCH_RigDispatch_SetModuleMetadata(in Name,in NameSpace,in Value,out Success)"}
   > RigVMDispatchNode(Name=Suffix, NameSpace=Self, Value=@n2.Value.ControlSuffix_38_#6) {template="DISPATCH_RigDispatch_SetModuleMetadata(in Name,in NameSpace,in Value,out Success)"}
   > RigVMDispatchNode(Name=Left Side Suffix, NameSpace=Self, Value=@n2.Value.LeftSideSuffix_23_#7) {template="DISPATCH_RigDispatch_SetModuleMetadata(in Name,in NameSpace,in Value,out Success)"}
   > RigVMDispatchNode(Name=Right Side Suffix, NameSpace=Self, Value=@n2.Value.RightSideSuffix_24_#8) {template="DISPATCH_RigDispatch_SetModuleMetadata(in Name,in NameSpace,in Value,out Success)"}
   > RigVMDispatchNode(Name=Mirror Axis, NameSpace=Self, Value=@n2.Value.LeftAxis_52_#9) {template="DISPATCH_RigDispatch_SetModuleMetadata(in Name,in NameSpace,in Value,out Success)"}
   > RigVMDispatchNode(Name=Debug, NameSpace=Self, Value=RigVMVariableNode() {variable=Debug}.Value) {template="DISPATCH_RigDispatch_SetModuleMetadata(in Name,in NameSpace,in Value,out Success)"}
   > RigVMDispatchNode(Name=Settings Control, NameSpace=Self, Value=RigVMVariableNode() {variable=global control}.Value) {template="DISPATCH_RigDispatch_SetModuleMetadata(in Name,in NameSpace,in Value,out Success)"}
   > RigVMDispatchNode(Name=Parent Control, NameSpace=Self, Value=RigVMVariableNode() {variable=body offset control}.Value) {template="DISPATCH_RigDispatch_SetModuleMetadata(in Name,in NameSpace,in Value,out Success)"}
   > Set_Module_Metadata_2 = RigVMDispatchNode(
    Name=Root Bone
    NameSpace=Self
    Value=&RigUnit_ResolveConnector(Connector=(Type=Connector,Name="Root"), SkipSocket=true).Result
   ) {template="DISPATCH_RigDispatch_SetModuleMetadata(in Name,in NameSpace,in Value,out Success)"}
   > Set_Module_Metadata_3 = RigVMDispatchNode(Name=Root Control, NameSpace=Self, Value=RigVMVariableNode() {variable=Root Control}.Value) {template="DISPATCH_RigDispatch_SetModuleMetadata(in Name,in NameSpace,in Value,out Success)"}
   Get_Module_Metadata_5 = RigVMDispatchNode(Default=(X=0,Y=0,Z=0), Name=Mirror Axis, NameSpace=Self) {template="DISPATCH_RigDispatch_GetModuleMetadata(in Name,in NameSpace,in Default,out Value,out Found)"}
   RigVMDispatchNode(
    Condition=@Get_Module_Metadata_5.Found
    True=@Get_Module_Metadata_5.Value
    False=@n2.Value.PrimaryBoneAxis_32_#10
   ) {template="DISPATCH_RigVMDispatch_If(in Condition,in True,in False,out Result)"}
   RigVMFunctionReturnNode(Value=@RigVMFunction_Sequence.B)
  ]}
 CollapseNode_ContainedGraph {nodes [
   n1 = RigVMFunctionEntryNode()
   RigVMFunction_Sequence = &RigVMFunction_Sequence(ExecuteContext=@n1.ExecuteContext)
   n2 = &RigVMFunction_ControlFlowBranch(ExecuteContext=@RigVMFunction_Sequence.A, Condition=RigVMVariableNode() {variable=Generate Sockets}.Value)
   Get BBox of Skeleton = RigVMFunctionReferenceNode(Offset=(X=0,Y=0,Z=50), Use Max Box Size=true)
   Create Socket at Metadata_2 = RigVMFunctionReferenceNode(
    Color=(R=0.086,G=0.694,B=0.702,A=1)
    Description=Drag and drop the BipedPhysics module here . . .
    Extra Transform=(Rotation=(X=0,Y=0,Z=0,W=1),Translation=(X=0,Y=0,Z=250),Scale3D=(X=1,Y=1,Z=1))
    Metadata Name=Root Bone
    Name=biped_physics_socket
    Space=GlobalSpace
    Extra Transform.Translation.Z=@"Get BBox of Skeleton".Final Position.Z
    ExecuteContext=@n2.Completed
   )
   Get_Module_Metadata_1 = RigVMDispatchNode(Default=(Type=Bone,Name="None"), Name=Pelvis Bone, NameSpace=Self) {template="DISPATCH_RigDispatch_GetModuleMetadata(in Name,in NameSpace,in Default,out Value,out Found)"}
   Create Socket at Metadata_1 = RigVMFunctionReferenceNode(
    Color=(R=1,G=1,B=1,A=1)
    Description=Drag and drop the Spine module here . . .
    Extra Transform=(Rotation=(X=0,Y=0,Z=0,W=1),Translation=(X=0,Y=0,Z=0),Scale3D=(X=1,Y=1,Z=1))
    Metadata Name=Pelvis Bone
    Space=LocalSpace
    Name=&RigVMFunction_NameConcat(B=_socket, A=@Get_Module_Metadata_1.Value.Name) {template="Concat::Execute(in A,in B,out Result)"}.Result
    ExecuteContext=@n2.True
   )
   > n3 = RigVMVariableNode(Value=true) {variable=Socket Built}
   n4 = &RigVMFunction_ControlFlowBranch(Condition=&RigVMFunction_MathBoolNot(Value=@"Create Socket at Metadata_1".Found).Result, ExecuteContext=@n3.ExecuteContext)
   n5 = &RigVMFunction_ControlFlowBranch(
    ExecuteContext=@n4.Completed
    Condition=&RigVMFunction_MathBoolNot(Value=RigVMVariableNode() {variable=Socket Built}.Value).Result
   )
   For_Each = RigVMDispatchNode(
    Array=/Script/ControlRig.RigUnit_CollectionChildrenArray(
     TypeToSearch=Bone
     bDefaultChildren=True
     bIncludeParent=False
     bRecursive=False
     Parent=&RigUnit_ResolveConnector(Connector=(Type=Connector,Name="Root"), SkipSocket=true).Result
    ) {is=struct RigVMUnitNode}.Items
    ExecuteContext=@n5.True
   ) {template="DISPATCH_RigVMDispatch_ArrayIterator(in Array,out Element,out Index,out Count,out Ratio)"}
   Concat = &RigVMFunction_NameConcat(B=_socket, A=@For_Each.Element.Name) {template="Concat::Execute(in A,in B,out Result)"}
   > /Script/ControlRig.RigUnit_HierarchyAddSocket(
    Color=(R=1,G=1,B=1,A=1)
    Description="Drag and drop module here . . . "
    Space=LocalSpace
    Transform=(Rotation=(X=0,Y=0,Z=0,W=1),Translation=(X=0,Y=0,Z=0),Scale3D=(X=1,Y=1,Z=1))
    Name=@Concat.Result
    Parent=@For_Each.Element
   ) {is=struct RigVMUnitNode}
   n6 = RigVMDispatchNode(ExecuteContext=@n4.True, Array=RigVMVariableNode() {variable=Spine Names}.Value) {template="DISPATCH_RigVMDispatch_ArrayIterator(in Array,out Element,out Index,out Count,out Ratio)"}
   > Find first in Hierarchy_1 = RigVMFunctionReferenceNode(
    Children=true
    Ignore=()
    Search for=("None")
    Type to Search=Bone
    Search for.0=@n6.Element
    Item=&RigUnit_ResolveConnector(Connector=(Type=Connector,Name="Root"), SkipSocket=true).Result
   )
   n7 = &RigVMFunction_ControlFlowBranch(
    ExecuteContext=@"Find first in Hierarchy_1".ExecuteContext
    Condition=/Script/ControlRig.RigUnit_ItemExists(Item=@"Find first in Hierarchy_1".Item Found) {is=struct RigVMUnitNode}.Exists
   )
   &RigUnit_HierarchyAddSocket(
    Color=(R=1,G=1,B=1,A=1)
    Description=Drag and drop the Spine module here . . .
    Space=LocalSpace
    Transform=(Rotation=(X=0,Y=0,Z=0,W=1),Translation=(X=0,Y=0,Z=0),Scale3D=(X=1,Y=1,Z=1))
    Name=&RigVMFunction_NameConcat(B=_socket, A=@"Find first in Hierarchy_1".Item Found.Name) {template="Concat::Execute(in A,in B,out Result)"}.Result
    Parent=@"Find first in Hierarchy_1".Item Found
    ExecutePin=@n7.True
   )
   > RigVMVariableNode(Value=true) {variable=Socket Built}
   n8 = &RigVMFunction_ControlFlowBranch(
    ExecuteContext=@n6.Completed
    Condition=&RigVMFunction_MathBoolNot(Value=RigVMVariableNode() {variable=Socket Built}.Value).Result
   )
   n9 = &RigUnit_CollectionChildrenArray(
    TypeToSearch=Bone
    bDefaultChildren=True
    bIncludeParent=False
    bRecursive=False
    Parent=&RigUnit_ResolveConnector(Connector=(Type=Connector,Name="Root"), SkipSocket=true).Result
   )
   For_Each_1 = RigVMDispatchNode(Array=@n9.Items, ExecuteContext=@n8.True) {template="DISPATCH_RigVMDispatch_ArrayIterator(in Array,out Element,out Index,out Count,out Ratio)"}
   > &RigUnit_HierarchyAddSocket(
    Color=(R=1,G=1,B=1,A=1)
    Description="Drag and drop module here . . . "
    Name=spine_01_socket
    Space=LocalSpace
    Transform=(Rotation=(X=0,Y=0,Z=0,W=1),Translation=(X=0,Y=0,Z=0),Scale3D=(X=1,Y=1,Z=1))
    Parent=@For_Each_1.Element
   )
   > RigVMVariableNode(Value=true) {variable=Socket Built}
   RigVMFunctionReturnNode(ExecuteContext=@RigVMFunction_Sequence.B)
   &RigVMFunction_NameConcat(
    B=_socket
    A=RigVMDispatchNode(Array=@n9.Items, Index=@For_Each_1.Index) {template="DISPATCH_RigVMDispatch_ArrayGetAtIndex(in Array,in Index,out Element)"}.Element.Name
   ) {template="Concat::Execute(in A,in B,out Result)"}
  ]}
 CollapseNode_ContainedGraph {nodes [
   RigVMFunctionEntryNode()
   Multiply_1 = RigVMAggregateNode(
    C=RigVMDispatchNode(Default=1, Name=Root Control Size, NameSpace=Root) {template="DISPATCH_RigDispatch_GetModuleMetadata(in Name,in NameSpace,in Default,out Value,out Found)"}.Value
    B=RigVMVariableNode() {variable=Root Control Scale}.Value
    A=RigVMVariableNode() {variable=Root Module Settings}.Value.ControlSize_12_#4
   )
   > Scale Module Controls_1 = RigVMFunctionReferenceNode(
    Global Scale=RigVMDispatchNode(Default=1, Name=Control Size, NameSpace=Root) {template="DISPATCH_RigDispatch_GetModuleMetadata(in Name,in NameSpace,in Default,out Value,out Found)"}.Value
    Controls=/Script/ControlRig.RigUnit_FindItemsWithMetadataTag(NameSpace=Self, Tag=Controls) {is=struct RigVMUnitNode}.Items
    Local Scale=@Multiply_1.Result
   )
   n1 = RigVMVariableNode() {variable=Root Module Settings}
   Axis to Bool_1 = RigVMFunctionReferenceNode(Value=@n1.Value.LeftAxis_52_#9)
   > ColorizeControls_1 = RigVMFunctionReferenceNode(
    ignore_ctrls=()
    Negative=@"Axis to Bool_1".Is Negative
    X Plane=@"Axis to Bool_1".X
    Y Plane=@"Axis to Bool_1".Y
    Z Plane=@"Axis to Bool_1".Z
    controls=&RigUnit_FindItemsWithMetadataTag(NameSpace=Self, Tag=Controls).Items
    CenterColor=@n1.Value.CenterColor_9_#1
    LeftColor=@n1.Value.LeftColor_3_#2
    RightColor=@n1.Value.RightColor_5_#3
   )
   > RigVMFunctionReturnNode()
   Get_Module_Metadata_1 = RigVMDispatchNode(Default=(X=0,Y=0,Z=0), Name=Mirror Axis, NameSpace=Self) {template="DISPATCH_RigDispatch_GetModuleMetadata(in Name,in NameSpace,in Default,out Value,out Found)"}
  ]}
 ContainedGraph {nodes [
   n1 = RigVMFunctionEntryNode()
   RigUnit_SequenceAggregate_1 = &RigVMFunction_Sequence(ExecuteContext=@n1.ExecuteContext)
   RigUnit_SequenceAggregate = RigVMAggregateNode(ExecuteContext=@RigUnit_SequenceAggregate_1.A)
   Greater = /Script/RigVM.RigVMFunction_MathDoubleGreater(B=0, A=@n1.Mirror Plane.X) {is=struct RigVMUnitNode, template="Greater::Execute(in A,in B,out Result)"}
   BranchNode_4 = &RigVMFunction_ControlFlowBranch(Condition=@Greater.Result, ExecuteContext=@RigUnit_SequenceAggregate.A)
   n2 = RigVMVariableNode() {variable=vector pos}
   RigVMVariableNode(
    Value=(X=0,Y=0,Z=0)
    ExecuteContext=@BranchNode_4.True
    Value.X=&RigUnit_GetTransform(bInitial=true, Item=@n1.Item).Transform.Translation.X
    Value.Y=@n2.Value.Y
    Value.Z=@n2.Value.Z
   ) {variable=vector pos}
   n3 = &RigVMFunction_MathDoubleGreater(B=0, A=@n1.Mirror Plane.Y) {template="Greater::Execute(in A,in B,out Result)"}
   BranchNode_1_1 = &RigVMFunction_ControlFlowBranch(Condition=@n3.Result, ExecuteContext=@RigUnit_SequenceAggregate.B)
   n4 = RigVMVariableNode() {variable=vector pos}
   RigVMVariableNode(
    Value=(X=0,Y=0,Z=0)
    ExecuteContext=@BranchNode_1_1.True
    Value.Y=&RigUnit_GetTransform(bInitial=true, Item=@n1.Item).Transform.Translation.Y
    Value.X=@n4.Value.X
    Value.Z=@n4.Value.Z
   ) {variable=vector pos}
   n5 = &RigVMFunction_MathDoubleGreater(B=0, A=@n1.Mirror Plane.Z) {template="Greater::Execute(in A,in B,out Result)"}
   BranchNode_2_1 = &RigVMFunction_ControlFlowBranch(Condition=@n5.Result, ExecuteContext=@RigUnit_SequenceAggregate.C)
   n6 = RigVMVariableNode() {variable=vector pos}
   RigVMVariableNode(
    Value=(X=0,Y=0,Z=0)
    ExecuteContext=@BranchNode_2_1.True
    Value.Z=&RigUnit_GetTransform(bInitial=true, Item=@n1.Item).Transform.Translation.Z
    Value.X=@n6.Value.X
    Value.Y=@n6.Value.Y
   ) {variable=vector pos}
   Cast = &RigVMFunction_MathBoolToFloat(Value=@Greater.Result)
   Dot__ = /Script/RigVM.RigVMFunction_MathVectorDot(
    B=(X=1,Y=1,Z=0)
    B.X=@Cast.Result
    B.Y=&RigVMFunction_MathBoolToFloat(Value=@n3.Result).Result
    B.Z=&RigVMFunction_MathBoolToFloat(Value=@n5.Result).Result
    A=RigVMVariableNode() {variable=vector pos}.Value
   ) {is=struct RigVMUnitNode, template="Dot,|::Execute(in A,in B,out Result)"}
   Less = /Script/RigVM.RigVMFunction_MathDoubleLess(B=0, A=@Dot__.Result) {is=struct RigVMUnitNode, template="Less::Execute(in A,in B,out Result)"}
   RigVMVariableNode(Value=@Less.Result, ExecuteContext=@RigUnit_SequenceAggregate.D) {variable=is on right}
   RigVMFunctionReturnNode(ExecuteContext=@RigUnit_SequenceAggregate_1.B, Right Side=RigVMVariableNode() {variable=is on right}.Value)
  ]}
 ContainedGraph {nodes [
   n1 = RigVMFunctionEntryNode()
   RigUnit_SequenceAggregate = &RigVMFunction_Sequence(ExecuteContext=@n1.ExecuteContext)
   RigUnit_SequenceAggregate_1 = &RigVMFunction_Sequence(ExecuteContext=@RigUnit_SequenceAggregate.B)
   RigUnit_SequenceAggregate_2 = &RigVMFunction_Sequence(ExecuteContext=@RigUnit_SequenceAggregate_1.B)
   RigVMFunctionReturnNode(A=@RigUnit_SequenceAggregate.A, B=@RigUnit_SequenceAggregate_1.A, C=@RigUnit_SequenceAggregate_2.A, D=@RigUnit_SequenceAggregate_2.B)
  ]}
 Multiply_2_1_ContainedGraph {nodes [
   n1 = RigVMFunctionEntryNode()
   RigVMFunctionReturnNode(Result=&RigVMFunction_MathDoubleMul(B=@n1.C, A=&RigVMFunction_MathDoubleMul(A=@n1.A, B=@n1.B).Result).Result)
  ]}
 New Function_ContainedGraph {nodes [
   RigVMFunctionEntryNode()
   > RigVMFunction_Sequence = RigVMAggregateNode()
   DISPATCH_RigVMDispatch_ArrayIterator = RigVMDispatchNode(Array=&RigUnit_FindItemsWithMetadataTag(NameSpace=Self, Tag=Controls).Items, ExecuteContext=@RigVMFunction_Sequence.A) {template="DISPATCH_RigVMDispatch_ArrayIterator(in Array,out Element,out Index,out Count,out Ratio)"}
   DISPATCH_RigDispatch_GetMetadata = RigVMDispatchNode(Default=(Type=Bone,Name="None"), Name=Item, NameSpace=Self, Item=@DISPATCH_RigVMDispatch_ArrayIterator.Element) {template="DISPATCH_RigDispatch_GetMetadata(in Item,in Name,in NameSpace,in Default,out Value,out Found)"}
   RigVMFunction_ControlFlowBranch = &RigVMFunction_ControlFlowBranch(Condition=@DISPATCH_RigDispatch_GetMetadata.Found, ExecuteContext=@DISPATCH_RigVMDispatch_ArrayIterator.ExecuteContext)
   &RigUnit_SetTransform(
    Weight=1
    bInitial=False
    Item=@DISPATCH_RigVMDispatch_ArrayIterator.Element
    Value=/Script/ControlRig.RigUnit_ProjectTransformToNewParent(
     bChildInitial=True
     bNewParentInitial=False
     bOldParentInitial=True
     NewParent=@DISPATCH_RigDispatch_GetMetadata.Value
     OldParent=@DISPATCH_RigDispatch_GetMetadata.Value
     Child=@DISPATCH_RigVMDispatch_ArrayIterator.Element
    ) {is=struct RigVMUnitNode}.Transform
    ExecutePin=@RigVMFunction_ControlFlowBranch.True
   )
   RigVMFunctionReturnNode(ExecuteContext=@RigVMFunction_Sequence.C)
  ]}
 New Function_ContainedGraph {nodes [
   RigVMFunctionEntryNode()
   > RigVMFunction_Sequence = RigVMAggregateNode()
   RigVMFunction_Sequence_2 = RigVMAggregateNode(ExecuteContext=@RigVMFunction_Sequence.A)
   Make_Array_1 = RigVMDispatchNode(Values=("Root_M","Hip_L_50")) {template="DISPATCH_RigVMDispatch_ArrayMake(in Values,out Array)"}
   RigVMFunction_MathBoolAnd_1 = RigVMAggregateNode(
    B=&RigUnit_ItemExists(
     Item=(Type=Bone,Name="None")
     Item.Name=RigVMDispatchNode(Index=1, Array=@Make_Array_1.Array) {template="DISPATCH_RigVMDispatch_ArrayGetAtIndex(in Array,in Index,out Element)"}.Element
    ).Exists
    A=&RigUnit_ItemExists(
     Item=(Type=Bone,Name="None")
     Item.Name=RigVMDispatchNode(Index=0, Array=@Make_Array_1.Array) {template="DISPATCH_RigVMDispatch_ArrayGetAtIndex(in Array,in Index,out Element)"}.Element
    ).Exists
    C=&RigVMFunction_MathBoolNot(Value=RigVMVariableNode() {variable=SkeletonFound}.Value).Result
   )
   n1 = &RigVMFunction_ControlFlowBranch(Condition=@RigVMFunction_MathBoolAnd_1.Result, ExecuteContext=@RigVMFunction_Sequence_2.C)
   RigVMVariableNode(Value=NewEnumerator3, ExecuteContext=@n1.True) {variable=Skeleton}
   > RigVMVariableNode(Value=true) {variable=SkeletonFound}
   From_String_2_1 = RigVMDispatchNode(String=Hip_L_00) {template="DISPATCH_RigDispatch_FromString(in String,out Result)"}
   > Set_Module_Metadata_3 = RigVMDispatchNode(Name=Pelvis Bone, NameSpace=Root, Value=(Type=Bone,Name="None"), Value.Name=@From_String_2_1.Result) {template="DISPATCH_RigDispatch_SetModuleMetadata(in Name,in NameSpace,in Value,out Success)"}
   From_String_1_1 = RigVMDispatchNode(String=RootPart1_M) {template="DISPATCH_RigDispatch_FromString(in String,out Result)"}
   > Set_Module_Metadata_1_1 = RigVMDispatchNode(Name=Spine Start Bone, NameSpace=Root, Value=(Type=Bone,Name="None"), Value.Name=@From_String_1_1.Result) {template="DISPATCH_RigDispatch_SetModuleMetadata(in Name,in NameSpace,in Value,out Success)"}
   From_String_4 = RigVMDispatchNode(String=Spine1Part1_M) {template="DISPATCH_RigDispatch_FromString(in String,out Result)"}
   > Set_Module_Metadata_2_2_1 = RigVMDispatchNode(Name=Spine End Bone, NameSpace=Root, Value=(Type=Bone,Name="None"), Value.Name=@From_String_4.Result) {template="DISPATCH_RigDispatch_SetModuleMetadata(in Name,in NameSpace,in Value,out Success)"}
   From_String_3_2 = RigVMDispatchNode(String=Shoulder_L) {template="DISPATCH_RigDispatch_FromString(in String,out Result)"}
   > Set_Module_Metadata_2_1_2 = RigVMDispatchNode(Name=shoulder_l, NameSpace=Root, Value=(Type=Bone,Name="None"), Value.Name=@From_String_3_2.Result) {template="DISPATCH_RigDispatch_SetModuleMetadata(in Name,in NameSpace,in Value,out Success)"}
   From_String_3_1_4 = RigVMDispatchNode(String=Shoulder_R) {template="DISPATCH_RigDispatch_FromString(in String,out Result)"}
   > Set_Module_Metadata_2_1_1_2 = RigVMDispatchNode(Name=shoulder_r, NameSpace=Root, Value=(Type=Bone,Name="None"), Value.Name=@From_String_3_1_4.Result) {template="DISPATCH_RigDispatch_SetModuleMetadata(in Name,in NameSpace,in Value,out Success)"}
   From_String_3_1_1_1 = RigVMDispatchNode(String=Hip_L) {template="DISPATCH_RigDispatch_FromString(in String,out Result)"}
   > Set_Module_Metadata_2_1_1_1_2 = RigVMDispatchNode(Name=thigh_l, NameSpace=Root, Value=(Type=Bone,Name="None"), Value.Name=@From_String_3_1_1_1.Result) {template="DISPATCH_RigDispatch_SetModuleMetadata(in Name,in NameSpace,in Value,out Success)"}
   From_String_3_1_2_1 = RigVMDispatchNode(String=Hip_R) {template="DISPATCH_RigDispatch_FromString(in String,out Result)"}
   > Set_Module_Metadata_2_1_1_1_1_2 = RigVMDispatchNode(Name=thigh_r, NameSpace=Root, Value=(Type=Bone,Name="None"), Value.Name=@From_String_3_1_2_1.Result) {template="DISPATCH_RigDispatch_SetModuleMetadata(in Name,in NameSpace,in Value,out Success)"}
   From_String_3_1_1_1_1 = RigVMDispatchNode(String=Knee_L) {template="DISPATCH_RigDispatch_FromString(in String,out Result)"}
   > Set_Module_Metadata_2_1_1_1_2_1 = RigVMDispatchNode(Name=knee_l, NameSpace=Root, Value=(Type=Bone,Name="None"), Value.Name=@From_String_3_1_1_1_1.Result) {template="DISPATCH_RigDispatch_SetModuleMetadata(in Name,in NameSpace,in Value,out Success)"}
   From_String_3_1_2_1_1 = RigVMDispatchNode(String=Knee_R) {template="DISPATCH_RigDispatch_FromString(in String,out Result)"}
   > Set_Module_Metadata_2_1_1_1_1_2_1 = RigVMDispatchNode(Name=knee_r, NameSpace=Root, Value=(Type=Bone,Name="None"), Value.Name=@From_String_3_1_2_1_1.Result) {template="DISPATCH_RigDispatch_SetModuleMetadata(in Name,in NameSpace,in Value,out Success)"}
   From_String_3_1_3_2 = RigVMDispatchNode(String=NeckPart1_M) {template="DISPATCH_RigDispatch_FromString(in String,out Result)"}
   > Set_Module_Metadata_2_1_1_1_1_1_2 = RigVMDispatchNode(Name=neck, NameSpace=Root, Value=(Type=Bone,Name="None"), Value.Name=@From_String_3_1_3_2.Result) {template="DISPATCH_RigDispatch_SetModuleMetadata(in Name,in NameSpace,in Value,out Success)"}
   > Set_Module_Metadata_2_6_1 = RigVMDispatchNode(Name=Arm PV Position, NameSpace=Root, Value=10) {template="DISPATCH_RigDispatch_SetModuleMetadata(in Name,in NameSpace,in Value,out Success)"}
   > Set_Module_Metadata_2_3_3_1 = RigVMDispatchNode(Name=Leg PV Position, NameSpace=Root, Value=15) {template="DISPATCH_RigDispatch_SetModuleMetadata(in Name,in NameSpace,in Value,out Success)"}
   From_String_3_1_3_2_1 = RigVMDispatchNode(String=Head_M) {template="DISPATCH_RigDispatch_FromString(in String,out Result)"}
   > Set_Module_Metadata_2_1_1_1_1_1_2_1 = RigVMDispatchNode(Name=head, NameSpace=Root, Value=(Type=Bone,Name="None"), Value.Name=@From_String_3_1_3_2_1.Result) {template="DISPATCH_RigDispatch_SetModuleMetadata(in Name,in NameSpace,in Value,out Success)"}
   From_String_3_1_3_1_2 = RigVMDispatchNode(String=Ankle_L) {template="DISPATCH_RigDispatch_FromString(in String,out Result)"}
   > Set_Module_Metadata_2_1_1_1_1_1_1_2 = RigVMDispatchNode(Name=foot_l, NameSpace=Root, Value=(Type=Bone,Name="None"), Value.Name=@From_String_3_1_3_1_2.Result) {template="DISPATCH_RigDispatch_SetModuleMetadata(in Name,in NameSpace,in Value,out Success)"}
   From_String_3_1_3_1_1_1 = RigVMDispatchNode(String=Ankle_R) {template="DISPATCH_RigDispatch_FromString(in String,out Result)"}
   > Set_Module_Metadata_2_1_1_1_1_1_1_1_1 = RigVMDispatchNode(Name=foot_r, NameSpace=Root, Value=(Type=Bone,Name="None"), Value.Name=@From_String_3_1_3_1_1_1.Result) {template="DISPATCH_RigDispatch_SetModuleMetadata(in Name,in NameSpace,in Value,out Success)"}
   From_String_3_1_3_1_2_1 = RigVMDispatchNode(String=Wrist_L) {template="DISPATCH_RigDispatch_FromString(in String,out Result)"}
   > Set_Module_Metadata_2_1_1_1_1_1_1_2_1 = RigVMDispatchNode(Name=hand_l, NameSpace=Root, Value=(Type=Bone,Name="None"), Value.Name=@From_String_3_1_3_1_2_1.Result) {template="DISPATCH_RigDispatch_SetModuleMetadata(in Name,in NameSpace,in Value,out Success)"}
   From_String_3_1_3_1_1_1_1 = RigVMDispatchNode(String=Wrist_R) {template="DISPATCH_RigDispatch_FromString(in String,out Result)"}
   > Set_Module_Metadata_2_1_1_1_1_1_1_1_1_1 = RigVMDispatchNode(Name=hand_r, NameSpace=Root, Value=(Type=Bone,Name="None"), Value.Name=@From_String_3_1_3_1_1_1_1.Result) {template="DISPATCH_RigDispatch_SetModuleMetadata(in Name,in NameSpace,in Value,out Success)"}
   From_String_3_1_3_1_2_1_3 = RigVMDispatchNode(String=Elbow_L) {template="DISPATCH_RigDispatch_FromString(in String,out Result)"}
   > Set_Module_Metadata_2_1_1_1_1_1_1_2_1_3 = RigVMDispatchNode(Name=elbow_l, NameSpace=Root, Value=(Type=Bone,Name="None"), Value.Name=@From_String_3_1_3_1_2_1_3.Result) {template="DISPATCH_RigDispatch_SetModuleMetadata(in Name,in NameSpace,in Value,out Success)"}
   From_String_3_1_3_1_1_1_1_3 = RigVMDispatchNode(String=Elbow_R) {template="DISPATCH_RigDispatch_FromString(in String,out Result)"}
   > Set_Module_Metadata_2_1_1_1_1_1_1_1_1_1_3 = RigVMDispatchNode(Name=elbow_r, NameSpace=Root, Value=(Type=Bone,Name="None"), Value.Name=@From_String_3_1_3_1_1_1_1_3.Result) {template="DISPATCH_RigDispatch_SetModuleMetadata(in Name,in NameSpace,in Value,out Success)"}
   From_String_3_1_3_1_1_1_1_3_1 = RigVMDispatchNode(String=RingFinger4_L) {template="DISPATCH_RigDispatch_FromString(in String,out Result)"}
   From_String_3_1_3_1_1_1_1_3_1_1 = RigVMDispatchNode(String=PinkyFinger4_L) {template="DISPATCH_RigDispatch_FromString(in String,out Result)"}
   From_String_3_1_3_1_1_1_1_3_1_1_2 = RigVMDispatchNode(String=PinkyFinger4_R) {template="DISPATCH_RigDispatch_FromString(in String,out Result)"}
   From_String_3_1_3_1_1_1_1_3_1_3 = RigVMDispatchNode(String=RingFinger4_R) {template="DISPATCH_RigDispatch_FromString(in String,out Result)"}
   From_String_3_1_3_1_1_1_1_4 = RigVMDispatchNode(String=IndexFinger4_L) {template="DISPATCH_RigDispatch_FromString(in String,out Result)"}
   From_String_3_1_3_1_1_1_1_4_2 = RigVMDispatchNode(String=IndexFinger4_R) {template="DISPATCH_RigDispatch_FromString(in String,out Result)"}
   From_String_3_1_3_1_1_1_2 = RigVMDispatchNode(String=ThumbFinger4_L) {template="DISPATCH_RigDispatch_FromString(in String,out Result)"}
   From_String_3_1_3_1_1_1_2_2 = RigVMDispatchNode(String=ThumbFinger4_R) {template="DISPATCH_RigDispatch_FromString(in String,out Result)"}
   From_String_3_1_3_1_2_1_3_1 = RigVMDispatchNode(String=MiddleFinger4_L) {template="DISPATCH_RigDispatch_FromString(in String,out Result)"}
   From_String_3_1_3_1_2_1_3_1_2 = RigVMDispatchNode(String=MiddleFinger4_R) {template="DISPATCH_RigDispatch_FromString(in String,out Result)"}
   > Set_Module_Metadata_2_1_1_1_1_1_1_1_1_1_3_2 = RigVMDispatchNode(
    Name=Finger Sockets
    NameSpace=Root
    Value=((Type=Bone,Name="None"),(Type=Bone,Name="None"),(Type=Bone,Name="None"),(Type=Bone,Name="None"),(Type=Bone,Name="None"),(Type=Bone,Name="None"),(Type=Bone,Name="None"),(Type=Bone,Name="None"),(Type=Bone,Name="None"),(Type=Bone,Name="None"))
    Value.6.Name=@From_String_3_1_3_1_1_1_1_3_1.Result
    Value.8.Name=@From_String_3_1_3_1_1_1_1_3_1_1.Result
    Value.9.Name=@From_String_3_1_3_1_1_1_1_3_1_1_2.Result
    Value.7.Name=@From_String_3_1_3_1_1_1_1_3_1_3.Result
    Value.2.Name=@From_String_3_1_3_1_1_1_1_4.Result
    Value.3.Name=@From_String_3_1_3_1_1_1_1_4_2.Result
    Value.0.Name=@From_String_3_1_3_1_1_1_2.Result
    Value.1.Name=@From_String_3_1_3_1_1_1_2_2.Result
    Value.4.Name=@From_String_3_1_3_1_2_1_3_1.Result
    Value.5.Name=@From_String_3_1_3_1_2_1_3_1_2.Result
   ) {template="DISPATCH_RigDispatch_SetModuleMetadata(in Name,in NameSpace,in Value,out Success)"}
   > Set_Module_Metadata_2_3_2_1 = RigVMDispatchNode(Name=Primary Axis, NameSpace=Root, Value=(X=-1,Y=0,Z=0)) {template="DISPATCH_RigDispatch_SetModuleMetadata(in Name,in NameSpace,in Value,out Success)"}
   > Set_Module_Metadata_2_3_2_1_1 = RigVMDispatchNode(Name=Secondary Axis, NameSpace=Root, Value=(X=0,Y=-1,Z=0)) {template="DISPATCH_RigDispatch_SetModuleMetadata(in Name,in NameSpace,in Value,out Success)"}
   > Set_Module_Metadata_5 = RigVMDispatchNode(Name=Enable Softness, NameSpace=Root, Value=False) {template="DISPATCH_RigDispatch_SetModuleMetadata(in Name,in NameSpace,in Value,out Success)"}
   > Set_Module_Metadata_2_3_2_1_4_1 = RigVMDispatchNode(Name=Mirror Axis, NameSpace=Root, Value=(X=1,Y=0,Z=0)) {template="DISPATCH_RigDispatch_SetModuleMetadata(in Name,in NameSpace,in Value,out Success)"}
   From_String_3_1_1_1_1_2_1 = RigVMDispatchNode(String=Toes_L) {template="DISPATCH_RigDispatch_FromString(in String,out Result)"}
   > Set_Module_Metadata_2_1_1_1_2_1_2_1 = RigVMDispatchNode(Name=toe_l, NameSpace=Root, Value=(Type=Bone,Name="None"), Value.Name=@From_String_3_1_1_1_1_2_1.Result) {template="DISPATCH_RigDispatch_SetModuleMetadata(in Name,in NameSpace,in Value,out Success)"}
   From_String_3_1_2_1_1_2_1 = RigVMDispatchNode(String=Toes_R) {template="DISPATCH_RigDispatch_FromString(in String,out Result)"}
   > Set_Module_Metadata_2_1_1_1_1_2_1_2_1 = RigVMDispatchNode(Name=toe_r, NameSpace=Root, Value=(Type=Bone,Name="None"), Value.Name=@From_String_3_1_2_1_1_2_1.Result) {template="DISPATCH_RigDispatch_SetModuleMetadata(in Name,in NameSpace,in Value,out Success)"}
   > Set_Module_Metadata_2_3_3_1_2 = RigVMDispatchNode(Name=Control Size, NameSpace=Root, Value=0.1) {template="DISPATCH_RigDispatch_SetModuleMetadata(in Name,in NameSpace,in Value,out Success)"}
   > Set_Module_Metadata_2_3_3_1_2_1 = RigVMDispatchNode(Name=Root Control Size, NameSpace=Root, Value=10) {template="DISPATCH_RigDispatch_SetModuleMetadata(in Name,in NameSpace,in Value,out Success)"}
   Make_Array_1_1 = RigVMDispatchNode(Values=("Dummy","s_genrig_RightFoot")) {template="DISPATCH_RigVMDispatch_ArrayMake(in Values,out Array)"}
   RigVMFunction_MathBoolAnd_1_1 = RigVMAggregateNode(
    B=&RigUnit_ItemExists(
     Item=(Type=Bone,Name="None")
     Item.Name=RigVMDispatchNode(Index=1, Array=@Make_Array_1_1.Array) {template="DISPATCH_RigVMDispatch_ArrayGetAtIndex(in Array,in Index,out Element)"}.Element
    ).Exists
    A=&RigUnit_ItemExists(
     Item=(Type=Bone,Name="None")
     Item.Name=RigVMDispatchNode(Index=0, Array=@Make_Array_1_1.Array) {template="DISPATCH_RigVMDispatch_ArrayGetAtIndex(in Array,in Index,out Element)"}.Element
    ).Exists
    C=&RigVMFunction_MathBoolNot(Value=RigVMVariableNode() {variable=SkeletonFound}.Value).Result
   )
   n2 = &RigVMFunction_ControlFlowBranch(Condition=@RigVMFunction_MathBoolAnd_1_1.Result, ExecuteContext=@RigVMFunction_Sequence_2.D)
   RigVMVariableNode(Value=NewEnumerator4, ExecuteContext=@n2.True) {variable=Skeleton}
   > RigVMVariableNode(Value=true) {variable=SkeletonFound}
   From_String_2_2 = RigVMDispatchNode(String=Hips) {template="DISPATCH_RigDispatch_FromString(in String,out Result)"}
   > Set_Module_Metadata_4 = RigVMDispatchNode(Name=Pelvis Bone, NameSpace=Root, Value=(Type=Bone,Name="None"), Value.Name=@From_String_2_2.Result) {template="DISPATCH_RigDispatch_SetModuleMetadata(in Name,in NameSpace,in Value,out Success)"}
   From_String_1_2 = RigVMDispatchNode(String=Spine) {template="DISPATCH_RigDispatch_FromString(in String,out Result)"}
   > Set_Module_Metadata_1_2 = RigVMDispatchNode(Name=Spine Start Bone, NameSpace=Root, Value=(Type=Bone,Name="None"), Value.Name=@From_String_1_2.Result) {template="DISPATCH_RigDispatch_SetModuleMetadata(in Name,in NameSpace,in Value,out Success)"}
   From_String_5 = RigVMDispatchNode(String=Spine3) {template="DISPATCH_RigDispatch_FromString(in String,out Result)"}
   > Set_Module_Metadata_2_2_2 = RigVMDispatchNode(Name=Spine End Bone, NameSpace=Root, Value=(Type=Bone,Name="None"), Value.Name=@From_String_5.Result) {template="DISPATCH_RigDispatch_SetModuleMetadata(in Name,in NameSpace,in Value,out Success)"}
   From_String_3_3 = RigVMDispatchNode(String=LeftArm) {template="DISPATCH_RigDispatch_FromString(in String,out Result)"}
   > Set_Module_Metadata_2_1_3 = RigVMDispatchNode(Name=shoulder_l, NameSpace=Root, Value=(Type=Bone,Name="None"), Value.Name=@From_String_3_3.Result) {template="DISPATCH_RigDispatch_SetModuleMetadata(in Name,in NameSpace,in Value,out Success)"}
   From_String_3_1_5 = RigVMDispatchNode(String=RightArm) {template="DISPATCH_RigDispatch_FromString(in String,out Result)"}
   > Set_Module_Metadata_2_1_1_3 = RigVMDispatchNode(Name=shoulder_r, NameSpace=Root, Value=(Type=Bone,Name="None"), Value.Name=@From_String_3_1_5.Result) {template="DISPATCH_RigDispatch_SetModuleMetadata(in Name,in NameSpace,in Value,out Success)"}
   From_String_3_1_1_2 = RigVMDispatchNode(String=LeftUpLeg) {template="DISPATCH_RigDispatch_FromString(in String,out Result)"}
   > Set_Module_Metadata_2_1_1_1_3 = RigVMDispatchNode(Name=thigh_l, NameSpace=Root, Value=(Type=Bone,Name="None"), Value.Name=@From_String_3_1_1_2.Result) {template="DISPATCH_RigDispatch_SetModuleMetadata(in Name,in NameSpace,in Value,out Success)"}
   From_String_3_1_2_2 = RigVMDispatchNode(String=RightUpLeg) {template="DISPATCH_RigDispatch_FromString(in String,out Result)"}
   > Set_Module_Metadata_2_1_1_1_1_3 = RigVMDispatchNode(Name=thigh_r, NameSpace=Root, Value=(Type=Bone,Name="None"), Value.Name=@From_String_3_1_2_2.Result) {template="DISPATCH_RigDispatch_SetModuleMetadata(in Name,in NameSpace,in Value,out Success)"}
   From_String_3_1_3_3 = RigVMDispatchNode(String=Neck) {template="DISPATCH_RigDispatch_FromString(in String,out Result)"}
   > Set_Module_Metadata_2_1_1_1_1_1_3 = RigVMDispatchNode(Name=neck, NameSpace=Root, Value=(Type=Bone,Name="None"), Value.Name=@From_String_3_1_3_3.Result) {template="DISPATCH_RigDispatch_SetModuleMetadata(in Name,in NameSpace,in Value,out Success)"}
   > Set_Module_Metadata_2_6 = RigVMDispatchNode(Name=Arm PV Position, NameSpace=Root, Value=50) {template="DISPATCH_RigDispatch_SetModuleMetadata(in Name,in NameSpace,in Value,out Success)"}
   > Set_Module_Metadata_2_3_3 = RigVMDispatchNode(Name=Leg PV Position, NameSpace=Root, Value=50) {template="DISPATCH_RigDispatch_SetModuleMetadata(in Name,in NameSpace,in Value,out Success)"}
   From_String_3_1_3_1_3 = RigVMDispatchNode(String=LeftFoot) {template="DISPATCH_RigDispatch_FromString(in String,out Result)"}
   > Set_Module_Metadata_2_1_1_1_1_1_1_3 = RigVMDispatchNode(Name=foot_l, NameSpace=Root, Value=(Type=Bone,Name="None"), Value.Name=@From_String_3_1_3_1_3.Result) {template="DISPATCH_RigDispatch_SetModuleMetadata(in Name,in NameSpace,in Value,out Success)"}
   From_String_3_1_3_1_1_2 = RigVMDispatchNode(String=s_genrig_RightFoot) {template="DISPATCH_RigDispatch_FromString(in String,out Result)"}
   > Set_Module_Metadata_2_1_1_1_1_1_1_1_2 = RigVMDispatchNode(Name=foot_r, NameSpace=Root, Value=(Type=Bone,Name="None"), Value.Name=@From_String_3_1_3_1_1_2.Result) {template="DISPATCH_RigDispatch_SetModuleMetadata(in Name,in NameSpace,in Value,out Success)"}
   From_String_3_1_3_1_2_1_2_1 = RigVMDispatchNode(String=LeftHand) {template="DISPATCH_RigDispatch_FromString(in String,out Result)"}
   > Set_Module_Metadata_2_1_1_1_1_1_1_2_1_2_1 = RigVMDispatchNode(Name=hand_l, NameSpace=Root, Value=(Type=Bone,Name="None"), Value.Name=@From_String_3_1_3_1_2_1_2_1.Result) {template="DISPATCH_RigDispatch_SetModuleMetadata(in Name,in NameSpace,in Value,out Success)"}
   From_String_3_1_3_1_1_1_1_2_1 = RigVMDispatchNode(String=RightHand) {template="DISPATCH_RigDispatch_FromString(in String,out Result)"}
   > Set_Module_Metadata_2_1_1_1_1_1_1_1_1_1_2_1 = RigVMDispatchNode(Name=hand_r, NameSpace=Root, Value=(Type=Bone,Name="None"), Value.Name=@From_String_3_1_3_1_1_1_1_2_1.Result) {template="DISPATCH_RigDispatch_SetModuleMetadata(in Name,in NameSpace,in Value,out Success)"}
   From_String_3_1_3_2_1_1 = RigVMDispatchNode(String=Head) {template="DISPATCH_RigDispatch_FromString(in String,out Result)"}
   > Set_Module_Metadata_2_1_1_1_1_1_2_1_1 = RigVMDispatchNode(Name=head, NameSpace=Root, Value=(Type=Bone,Name="None"), Value.Name=@From_String_3_1_3_2_1_1.Result) {template="DISPATCH_RigDispatch_SetModuleMetadata(in Name,in NameSpace,in Value,out Success)"}
   > Set_Module_Metadata_2_3_2_1_4_1_1 = RigVMDispatchNode(Name=Mirror Axis, NameSpace=Root, Value=(X=1,Y=0,Z=0)) {template="DISPATCH_RigDispatch_SetModuleMetadata(in Name,in NameSpace,in Value,out Success)"}
   From_String_3_1_1_1_1_2_1_1 = RigVMDispatchNode(String=LeftToeBase1) {template="DISPATCH_RigDispatch_FromString(in String,out Result)"}
   > Set_Module_Metadata_2_1_1_1_2_1_2_1_1 = RigVMDispatchNode(Name=toe_l, NameSpace=Root, Value=(Type=Bone,Name="None"), Value.Name=@From_String_3_1_1_1_1_2_1_1.Result) {template="DISPATCH_RigDispatch_SetModuleMetadata(in Name,in NameSpace,in Value,out Success)"}
   From_String_3_1_2_1_1_2_1_1 = RigVMDispatchNode(String=RightToeBase1) {template="DISPATCH_RigDispatch_FromString(in String,out Result)"}
   > Set_Module_Metadata_2_1_1_1_1_2_1_2_1_1 = RigVMDispatchNode(Name=toe_r, NameSpace=Root, Value=(Type=Bone,Name="None"), Value.Name=@From_String_3_1_2_1_1_2_1_1.Result) {template="DISPATCH_RigDispatch_SetModuleMetadata(in Name,in NameSpace,in Value,out Success)"}
   Make_Array_1_1_1 = RigVMDispatchNode(Values=("root_C0_0_jnt","spine_C0_01_jnt")) {template="DISPATCH_RigVMDispatch_ArrayMake(in Values,out Array)"}
   RigVMFunction_MathBoolAnd_1_1_1 = RigVMAggregateNode(
    B=&RigUnit_ItemExists(
     Item=(Type=Bone,Name="None")
     Item.Name=RigVMDispatchNode(Index=1, Array=@Make_Array_1_1_1.Array) {template="DISPATCH_RigVMDispatch_ArrayGetAtIndex(in Array,in Index,out Element)"}.Element
    ).Exists
    A=&RigUnit_ItemExists(
     Item=(Type=Bone,Name="None")
     Item.Name=RigVMDispatchNode(Index=0, Array=@Make_Array_1_1_1.Array) {template="DISPATCH_RigVMDispatch_ArrayGetAtIndex(in Array,in Index,out Element)"}.Element
    ).Exists
    C=&RigVMFunction_MathBoolNot(Value=RigVMVariableNode() {variable=SkeletonFound}.Value).Result
   )
   n3 = &RigVMFunction_ControlFlowBranch(Condition=@RigVMFunction_MathBoolAnd_1_1_1.Result, ExecuteContext=@RigVMFunction_Sequence_2.E)
   RigVMVariableNode(Value=NewEnumerator5, ExecuteContext=@n3.True) {variable=Skeleton}
   > RigVMVariableNode(Value=true) {variable=SkeletonFound}
   From_String_2 = RigVMDispatchNode(String=spine_C0_pelvis_jnt) {template="DISPATCH_RigDispatch_FromString(in String,out Result)"}
   > Set_Module_Metadata = RigVMDispatchNode(Name=Pelvis Bone, NameSpace=Root, Value=(Type=Bone,Name="None"), Value.Name=@From_String_2.Result) {template="DISPATCH_RigDispatch_SetModuleMetadata(in Name,in NameSpace,in Value,out Success)"}
   From_String_1 = RigVMDispatchNode(String=spine_C0_01_jnt) {template="DISPATCH_RigDispatch_FromString(in String,out Result)"}
   > Set_Module_Metadata_1 = RigVMDispatchNode(Name=Spine Start Bone, NameSpace=Root, Value=(Type=Bone,Name="None"), Value.Name=@From_String_1.Result) {template="DISPATCH_RigDispatch_SetModuleMetadata(in Name,in NameSpace,in Value,out Success)"}
   From_String = RigVMDispatchNode(String=spine_C0_04_jnt) {template="DISPATCH_RigDispatch_FromString(in String,out Result)"}
   > Set_Module_Metadata_2_2 = RigVMDispatchNode(Name=Spine End Bone, NameSpace=Root, Value=(Type=Bone,Name="None"), Value.Name=@From_String.Result) {template="DISPATCH_RigDispatch_SetModuleMetadata(in Name,in NameSpace,in Value,out Success)"}
   From_String_3 = RigVMDispatchNode(String=arm_L0_upperarm_jnt) {template="DISPATCH_RigDispatch_FromString(in String,out Result)"}
   > Set_Module_Metadata_2_1 = RigVMDispatchNode(Name=shoulder_l, NameSpace=Root, Value=(Type=Bone,Name="None"), Value.Name=@From_String_3.Result) {template="DISPATCH_RigDispatch_SetModuleMetadata(in Name,in NameSpace,in Value,out Success)"}
   From_String_3_1 = RigVMDispatchNode(String=arm_R0_upperarm_jnt) {template="DISPATCH_RigDispatch_FromString(in String,out Result)"}
   > Set_Module_Metadata_2_1_1 = RigVMDispatchNode(Name=shoulder_r, NameSpace=Root, Value=(Type=Bone,Name="None"), Value.Name=@From_String_3_1.Result) {template="DISPATCH_RigDispatch_SetModuleMetadata(in Name,in NameSpace,in Value,out Success)"}
   From_String_3_1_1 = RigVMDispatchNode(String=leg_L0_thigh_jnt) {template="DISPATCH_RigDispatch_FromString(in String,out Result)"}
   > Set_Module_Metadata_2_1_1_1 = RigVMDispatchNode(Name=thigh_l, NameSpace=Root, Value=(Type=Bone,Name="None"), Value.Name=@From_String_3_1_1.Result) {template="DISPATCH_RigDispatch_SetModuleMetadata(in Name,in NameSpace,in Value,out Success)"}
   From_String_3_1_2 = RigVMDispatchNode(String=leg_R0_thigh_jnt) {template="DISPATCH_RigDispatch_FromString(in String,out Result)"}
   > Set_Module_Metadata_2_1_1_1_1 = RigVMDispatchNode(Name=thigh_r, NameSpace=Root, Value=(Type=Bone,Name="None"), Value.Name=@From_String_3_1_2.Result) {template="DISPATCH_RigDispatch_SetModuleMetadata(in Name,in NameSpace,in Value,out Success)"}
   From_String_3_1_3 = RigVMDispatchNode(String=neck_C0_01_jnt) {template="DISPATCH_RigDispatch_FromString(in String,out Result)"}
   > Set_Module_Metadata_2_1_1_1_1_1 = RigVMDispatchNode(Name=neck, NameSpace=Root, Value=(Type=Bone,Name="None"), Value.Name=@From_String_3_1_3.Result) {template="DISPATCH_RigDispatch_SetModuleMetadata(in Name,in NameSpace,in Value,out Success)"}
   > Set_Module_Metadata_2 = RigVMDispatchNode(Name=Arm PV Position, NameSpace=Root, Value=50) {template="DISPATCH_RigDispatch_SetModuleMetadata(in Name,in NameSpace,in Value,out Success)"}
   > Set_Module_Metadata_2_3 = RigVMDispatchNode(Name=Leg PV Position, NameSpace=Root, Value=50) {template="DISPATCH_RigDispatch_SetModuleMetadata(in Name,in NameSpace,in Value,out Success)"}
   From_String_3_1_3_1 = RigVMDispatchNode(String=leg_L0_foot_jnt) {template="DISPATCH_RigDispatch_FromString(in String,out Result)"}
   > Set_Module_Metadata_2_1_1_1_1_1_1 = RigVMDispatchNode(Name=foot_l, NameSpace=Root, Value=(Type=Bone,Name="None"), Value.Name=@From_String_3_1_3_1.Result) {template="DISPATCH_RigDispatch_SetModuleMetadata(in Name,in NameSpace,in Value,out Success)"}
   From_String_3_1_3_1_1 = RigVMDispatchNode(String=leg_R0_foot_jnt) {template="DISPATCH_RigDispatch_FromString(in String,out Result)"}
   > Set_Module_Metadata_2_1_1_1_1_1_1_1 = RigVMDispatchNode(Name=foot_r, NameSpace=Root, Value=(Type=Bone,Name="None"), Value.Name=@From_String_3_1_3_1_1.Result) {template="DISPATCH_RigDispatch_SetModuleMetadata(in Name,in NameSpace,in Value,out Success)"}
   From_String_3_1_3_1_2_1_2 = RigVMDispatchNode(String=arm_L0_hand_jnt) {template="DISPATCH_RigDispatch_FromString(in String,out Result)"}
   > Set_Module_Metadata_2_1_1_1_1_1_1_2_1_2 = RigVMDispatchNode(Name=hand_l, NameSpace=Root, Value=(Type=Bone,Name="None"), Value.Name=@From_String_3_1_3_1_2_1_2.Result) {template="DISPATCH_RigDispatch_SetModuleMetadata(in Name,in NameSpace,in Value,out Success)"}
   From_String_3_1_3_1_1_1_1_2 = RigVMDispatchNode(String=arm_R0_hand_jnt) {template="DISPATCH_RigDispatch_FromString(in String,out Result)"}
   > Set_Module_Metadata_2_1_1_1_1_1_1_1_1_1_2 = RigVMDispatchNode(Name=hand_r, NameSpace=Root, Value=(Type=Bone,Name="None"), Value.Name=@From_String_3_1_3_1_1_1_1_2.Result) {template="DISPATCH_RigDispatch_SetModuleMetadata(in Name,in NameSpace,in Value,out Success)"}
   > Set_Module_Metadata_2_6_1_1 = RigVMDispatchNode(Name=Arm PV Position, NameSpace=Root, Value=100) {template="DISPATCH_RigDispatch_SetModuleMetadata(in Name,in NameSpace,in Value,out Success)"}
   > Set_Module_Metadata_2_3_2_1_4_1_1_1 = RigVMDispatchNode(Name=Mirror Axis, NameSpace=Root, Value=(X=1,Y=0,Z=0)) {template="DISPATCH_RigDispatch_SetModuleMetadata(in Name,in NameSpace,in Value,out Success)"}
   From_String_3_1_1_1_1_2_1_1_1 = RigVMDispatchNode(String=foot_L0_ball_jnt) {template="DISPATCH_RigDispatch_FromString(in String,out Result)"}
   > Set_Module_Metadata_2_1_1_1_2_1_2_1_1_1 = RigVMDispatchNode(Name=toe_l, NameSpace=Root, Value=(Type=Bone,Name="None"), Value.Name=@From_String_3_1_1_1_1_2_1_1_1.Result) {template="DISPATCH_RigDispatch_SetModuleMetadata(in Name,in NameSpace,in Value,out Success)"}
   From_String_3_1_2_1_1_2_1_1_1 = RigVMDispatchNode(String=foot_R0_ball_jnt) {template="DISPATCH_RigDispatch_FromString(in String,out Result)"}
   > Set_Module_Metadata_2_1_1_1_1_2_1_2_1_1_1 = RigVMDispatchNode(Name=toe_r, NameSpace=Root, Value=(Type=Bone,Name="None"), Value.Name=@From_String_3_1_2_1_1_2_1_1_1.Result) {template="DISPATCH_RigDispatch_SetModuleMetadata(in Name,in NameSpace,in Value,out Success)"}
   From_String_3_1_3_1_2_1_3_2_3 = RigVMDispatchNode(String=arm_L0_lowerarm_jnt) {template="DISPATCH_RigDispatch_FromString(in String,out Result)"}
   > Set_Module_Metadata_2_1_1_1_1_1_1_2_1_3_1_3 = RigVMDispatchNode(Name=elbow_l, NameSpace=Root, Value=(Type=Bone,Name="None"), Value.Name=@From_String_3_1_3_1_2_1_3_2_3.Result) {template="DISPATCH_RigDispatch_SetModuleMetadata(in Name,in NameSpace,in Value,out Success)"}
   From_String_3_1_3_1_1_1_1_3_2_3 = RigVMDispatchNode(String=arm_R0_lowerarm_jnt) {template="DISPATCH_RigDispatch_FromString(in String,out Result)"}
   > Set_Module_Metadata_2_1_1_1_1_1_1_1_1_1_3_1_3 = RigVMDispatchNode(Name=elbow_r, NameSpace=Root, Value=(Type=Bone,Name="None"), Value.Name=@From_String_3_1_3_1_1_1_1_3_2_3.Result) {template="DISPATCH_RigDispatch_SetModuleMetadata(in Name,in NameSpace,in Value,out Success)"}
   From_String_3_1_1_1_1_1_3 = RigVMDispatchNode(String=leg_L0_calf_jnt) {template="DISPATCH_RigDispatch_FromString(in String,out Result)"}
   > Set_Module_Metadata_2_1_1_1_2_1_1_3 = RigVMDispatchNode(Name=knee_l, NameSpace=Root, Value=(Type=Bone,Name="None"), Value.Name=@From_String_3_1_1_1_1_1_3.Result) {template="DISPATCH_RigDispatch_SetModuleMetadata(in Name,in NameSpace,in Value,out Success)"}
   From_String_3_1_2_1_1_1_3 = RigVMDispatchNode(String=leg_R0_calf_jnt) {template="DISPATCH_RigDispatch_FromString(in String,out Result)"}
   > Set_Module_Metadata_2_1_1_1_1_2_1_1_3 = RigVMDispatchNode(Name=knee_r, NameSpace=Root, Value=(Type=Bone,Name="None"), Value.Name=@From_String_3_1_2_1_1_1_3.Result) {template="DISPATCH_RigDispatch_SetModuleMetadata(in Name,in NameSpace,in Value,out Success)"}
   n4 = /Script/ControlRig.RigUnit_CollectionNameSearchArray(PartialName=spine_, TypeToSearch=All) {is=struct RigVMUnitNode}
   Make_Array_1_1_1_1 = RigVMDispatchNode(Values=("root","pelvis")) {template="DISPATCH_RigVMDispatch_ArrayMake(in Values,out Array)"}
   n5 = RigVMDispatchNode(Index=1, Array=@Make_Array_1_1_1_1.Array) {template="DISPATCH_RigVMDispatch_ArrayGetAtIndex(in Array,in Index,out Element)"}
   n6 = RigVMDispatchNode(Index=0, Array=@Make_Array_1_1_1_1.Array) {template="DISPATCH_RigVMDispatch_ArrayGetAtIndex(in Array,in Index,out Element)"}
   RigVMFunction_MathBoolAnd_2 = RigVMAggregateNode(
    B=RigVMDispatchNode(B=2, A=RigVMDispatchNode(Array=@n4.Items) {template="DISPATCH_RigVMDispatch_ArrayGetNum(in Array,out Num)"}.Num) {template="DISPATCH_RigVMDispatch_CoreEquals(in A,in B,out Result)"}.Result
    A=&RigVMFunction_MathBoolAnd(
     B=&RigUnit_ItemExists(Item=(Type=Bone,Name="None"), Item.Name=@n5.Element).Exists
     A=&RigUnit_ItemExists(Item=(Type=Bone,Name="None"), Item.Name=@n6.Element).Exists
    ).Result
    C=&RigVMFunction_MathBoolNot(Value=RigVMVariableNode() {variable=SkeletonFound}.Value).Result
   )
   n7 = &RigVMFunction_ControlFlowBranch(Condition=@RigVMFunction_MathBoolAnd_2.Result, ExecuteContext=@RigVMFunction_Sequence_2.F)
   RigVMVariableNode(Value=NewEnumerator6, ExecuteContext=@n7.True) {variable=Skeleton}
   > RigVMVariableNode(Value=true) {variable=SkeletonFound}
   From_String_2_1_1_1 = RigVMDispatchNode(String=pelvis) {template="DISPATCH_RigDispatch_FromString(in String,out Result)"}
   > Set_Module_Metadata_3_1_1 = RigVMDispatchNode(Name=Pelvis Bone, NameSpace=Root, Value=(Type=Bone,Name="None"), Value.Name=@From_String_2_1_1_1.Result) {template="DISPATCH_RigDispatch_SetModuleMetadata(in Name,in NameSpace,in Value,out Success)"}
   From_String_1_1_1_1 = RigVMDispatchNode(String=spine_01) {template="DISPATCH_RigDispatch_FromString(in String,out Result)"}
   > Set_Module_Metadata_1_1_1_1 = RigVMDispatchNode(Name=Spine Start Bone, NameSpace=Root, Value=(Type=Bone,Name="None"), Value.Name=@From_String_1_1_1_1.Result) {template="DISPATCH_RigDispatch_SetModuleMetadata(in Name,in NameSpace,in Value,out Success)"}
   From_String_4_1_1 = RigVMDispatchNode(String=spine_02) {template="DISPATCH_RigDispatch_FromString(in String,out Result)"}
   > Set_Module_Metadata_2_2_1_1_1 = RigVMDispatchNode(Name=Spine End Bone, NameSpace=Root, Value=(Type=Bone,Name="None"), Value.Name=@From_String_4_1_1.Result) {template="DISPATCH_RigDispatch_SetModuleMetadata(in Name,in NameSpace,in Value,out Success)"}
   From_String_3_2_1_1 = RigVMDispatchNode(String=upperarm_l) {template="DISPATCH_RigDispatch_FromString(in String,out Result)"}
   > Set_Module_Metadata_2_1_2_1_1 = RigVMDispatchNode(Name=shoulder_l, NameSpace=Root, Value=(Type=Bone,Name="None"), Value.Name=@From_String_3_2_1_1.Result) {template="DISPATCH_RigDispatch_SetModuleMetadata(in Name,in NameSpace,in Value,out Success)"}
   From_String_3_1_4_1_1 = RigVMDispatchNode(String=upperarm_r) {template="DISPATCH_RigDispatch_FromString(in String,out Result)"}
   > Set_Module_Metadata_2_1_1_2_1_1 = RigVMDispatchNode(Name=shoulder_r, NameSpace=Root, Value=(Type=Bone,Name="None"), Value.Name=@From_String_3_1_4_1_1.Result) {template="DISPATCH_RigDispatch_SetModuleMetadata(in Name,in NameSpace,in Value,out Success)"}
   From_String_3_1_1_1_2_1 = RigVMDispatchNode(String=thigh_l) {template="DISPATCH_RigDispatch_FromString(in String,out Result)"}
   > Set_Module_Metadata_2_1_1_1_2_2_1 = RigVMDispatchNode(Name=thigh_l, NameSpace=Root, Value=(Type=Bone,Name="None"), Value.Name=@From_String_3_1_1_1_2_1.Result) {template="DISPATCH_RigDispatch_SetModuleMetadata(in Name,in NameSpace,in Value,out Success)"}
   From_String_3_1_2_1_2_1 = RigVMDispatchNode(String=thigh_r) {template="DISPATCH_RigDispatch_FromString(in String,out Result)"}
   > Set_Module_Metadata_2_1_1_1_1_2_2_1 = RigVMDispatchNode(Name=thigh_r, NameSpace=Root, Value=(Type=Bone,Name="None"), Value.Name=@From_String_3_1_2_1_2_1.Result) {template="DISPATCH_RigDispatch_SetModuleMetadata(in Name,in NameSpace,in Value,out Success)"}
   From_String_3_1_1_1_1_1_1 = RigVMDispatchNode(String=calf_l) {template="DISPATCH_RigDispatch_FromString(in String,out Result)"}
   > Set_Module_Metadata_2_1_1_1_2_1_1_1 = RigVMDispatchNode(Name=knee_l, NameSpace=Root, Value=(Type=Bone,Name="None"), Value.Name=@From_String_3_1_1_1_1_1_1.Result) {template="DISPATCH_RigDispatch_SetModuleMetadata(in Name,in NameSpace,in Value,out Success)"}
   From_String_3_1_2_1_1_1_1 = RigVMDispatchNode(String=calf_r) {template="DISPATCH_RigDispatch_FromString(in String,out Result)"}
   > Set_Module_Metadata_2_1_1_1_1_2_1_1_1 = RigVMDispatchNode(Name=knee_r, NameSpace=Root, Value=(Type=Bone,Name="None"), Value.Name=@From_String_3_1_2_1_1_1_1.Result) {template="DISPATCH_RigDispatch_SetModuleMetadata(in Name,in NameSpace,in Value,out Success)"}
   From_String_3_1_3_2_2_1 = RigVMDispatchNode(String=neck_01) {template="DISPATCH_RigDispatch_FromString(in String,out Result)"}
   > Set_Module_Metadata_2_1_1_1_1_1_2_2_1 = RigVMDispatchNode(Name=neck, NameSpace=Root, Value=(Type=Bone,Name="None"), Value.Name=@From_String_3_1_3_2_2_1.Result) {template="DISPATCH_RigDispatch_SetModuleMetadata(in Name,in NameSpace,in Value,out Success)"}
   > Set_Module_Metadata_2_6_1_2_1 = RigVMDispatchNode(Name=Arm PV Position, NameSpace=Root, Value=10) {template="DISPATCH_RigDispatch_SetModuleMetadata(in Name,in NameSpace,in Value,out Success)"}
   > Set_Module_Metadata_2_3_3_1_1_1 = RigVMDispatchNode(Name=Leg PV Position, NameSpace=Root, Value=15) {template="DISPATCH_RigDispatch_SetModuleMetadata(in Name,in NameSpace,in Value,out Success)"}
   From_String_3_1_3_2_1_2_1 = RigVMDispatchNode(String=Head) {template="DISPATCH_RigDispatch_FromString(in String,out Result)"}
   > Set_Module_Metadata_2_1_1_1_1_1_2_1_2_1 = RigVMDispatchNode(Name=head, NameSpace=Root, Value=(Type=Bone,Name="None"), Value.Name=@From_String_3_1_3_2_1_2_1.Result) {template="DISPATCH_RigDispatch_SetModuleMetadata(in Name,in NameSpace,in Value,out Success)"}
   From_String_3_1_3_1_2_2_1 = RigVMDispatchNode(String=foot_l) {template="DISPATCH_RigDispatch_FromString(in String,out Result)"}
   > Set_Module_Metadata_2_1_1_1_1_1_1_2_2_1 = RigVMDispatchNode(Name=foot_l, NameSpace=Root, Value=(Type=Bone,Name="None"), Value.Name=@From_String_3_1_3_1_2_2_1.Result) {template="DISPATCH_RigDispatch_SetModuleMetadata(in Name,in NameSpace,in Value,out Success)"}
   From_String_3_1_3_1_1_1_3_1 = RigVMDispatchNode(String=foot_r) {template="DISPATCH_RigDispatch_FromString(in String,out Result)"}
   > Set_Module_Metadata_2_1_1_1_1_1_1_1_1_2_1 = RigVMDispatchNode(Name=foot_r, NameSpace=Root, Value=(Type=Bone,Name="None"), Value.Name=@From_String_3_1_3_1_1_1_3_1.Result) {template="DISPATCH_RigDispatch_SetModuleMetadata(in Name,in NameSpace,in Value,out Success)"}
   From_String_3_1_3_1_2_1_4_1 = RigVMDispatchNode(String=hand_l) {template="DISPATCH_RigDispatch_FromString(in String,out Result)"}
   > Set_Module_Metadata_2_1_1_1_1_1_1_2_1_4_1 = RigVMDispatchNode(Name=hand_l, NameSpace=Root, Value=(Type=Bone,Name="None"), Value.Name=@From_String_3_1_3_1_2_1_4_1.Result) {template="DISPATCH_RigDispatch_SetModuleMetadata(in Name,in NameSpace,in Value,out Success)"}
   From_String_3_1_3_1_1_1_1_5_1 = RigVMDispatchNode(String=hand_r) {template="DISPATCH_RigDispatch_FromString(in String,out Result)"}
   > Set_Module_Metadata_2_1_1_1_1_1_1_1_1_1_4_1 = RigVMDispatchNode(Name=hand_r, NameSpace=Root, Value=(Type=Bone,Name="None"), Value.Name=@From_String_3_1_3_1_1_1_1_5_1.Result) {template="DISPATCH_RigDispatch_SetModuleMetadata(in Name,in NameSpace,in Value,out Success)"}
   From_String_3_1_3_1_2_1_3_2_1 = RigVMDispatchNode(String=lowerarm_l) {template="DISPATCH_RigDispatch_FromString(in String,out Result)"}
   > Set_Module_Metadata_2_1_1_1_1_1_1_2_1_3_1_1 = RigVMDispatchNode(Name=elbow_l, NameSpace=Root, Value=(Type=Bone,Name="None"), Value.Name=@From_String_3_1_3_1_2_1_3_2_1.Result) {template="DISPATCH_RigDispatch_SetModuleMetadata(in Name,in NameSpace,in Value,out Success)"}
   From_String_3_1_3_1_1_1_1_3_2_1 = RigVMDispatchNode(String=lowerarm_r) {template="DISPATCH_RigDispatch_FromString(in String,out Result)"}
   > Set_Module_Metadata_2_1_1_1_1_1_1_1_1_1_3_1_1 = RigVMDispatchNode(Name=elbow_r, NameSpace=Root, Value=(Type=Bone,Name="None"), Value.Name=@From_String_3_1_3_1_1_1_1_3_2_1.Result) {template="DISPATCH_RigDispatch_SetModuleMetadata(in Name,in NameSpace,in Value,out Success)"}
   > Set_Module_Metadata_2_3_2_1_2 = RigVMDispatchNode(Name=Primary Axis, NameSpace=Root, Value=(X=0,Y=-1,Z=0)) {template="DISPATCH_RigDispatch_SetModuleMetadata(in Name,in NameSpace,in Value,out Success)"}
   > Set_Module_Metadata_2_3_2_1_1_1 = RigVMDispatchNode(Name=Secondary Axis, NameSpace=Root, Value=(X=1,Y=0,Z=0)) {template="DISPATCH_RigDispatch_SetModuleMetadata(in Name,in NameSpace,in Value,out Success)"}
   > Set_Module_Metadata_2_6_1_2_2 = RigVMDispatchNode(Name=Arm PV Position, NameSpace=Root, Value=100) {template="DISPATCH_RigDispatch_SetModuleMetadata(in Name,in NameSpace,in Value,out Success)"}
   > Set_Module_Metadata_2_3_3_1_1_2 = RigVMDispatchNode(Name=Leg PV Position, NameSpace=Root, Value=100) {template="DISPATCH_RigDispatch_SetModuleMetadata(in Name,in NameSpace,in Value,out Success)"}
   > Set_Module_Metadata_6_1 = RigVMDispatchNode(Name=elbow_bone_offset, NameSpace=Root, Value=(X=0,Y=-0.5,Z=0)) {template="DISPATCH_RigDispatch_SetModuleMetadata(in Name,in NameSpace,in Value,out Success)"}
   > Set_Module_Metadata_6_1_1 = RigVMDispatchNode(Name=knee_bone_offset, NameSpace=Root, Value=(X=0,Y=0.5,Z=0)) {template="DISPATCH_RigDispatch_SetModuleMetadata(in Name,in NameSpace,in Value,out Success)"}
   > Set_Module_Metadata_6_1_1_1 = RigVMDispatchNode(Name=knee_bone_offset_globalSpace, NameSpace=Root, Value=true) {template="DISPATCH_RigDispatch_SetModuleMetadata(in Name,in NameSpace,in Value,out Success)"}
   From_String_3_1_1_1_1_2_1_2 = RigVMDispatchNode(String=ball_l) {template="DISPATCH_RigDispatch_FromString(in String,out Result)"}
   > Set_Module_Metadata_2_1_1_1_2_1_2_1_2 = RigVMDispatchNode(Name=toe_l, NameSpace=Root, Value=(Type=Bone,Name="None"), Value.Name=@From_String_3_1_1_1_1_2_1_2.Result) {template="DISPATCH_RigDispatch_SetModuleMetadata(in Name,in NameSpace,in Value,out Success)"}
   From_String_3_1_2_1_1_2_1_2 = RigVMDispatchNode(String=ball_r) {template="DISPATCH_RigDispatch_FromString(in String,out Result)"}
   > Set_Module_Metadata_2_1_1_1_1_2_1_2_1_2 = RigVMDispatchNode(Name=toe_r, NameSpace=Root, Value=(Type=Bone,Name="None"), Value.Name=@From_String_3_1_2_1_1_2_1_2.Result) {template="DISPATCH_RigDispatch_SetModuleMetadata(in Name,in NameSpace,in Value,out Success)"}
   Make_Array_1_1_1_1_1 = RigVMDispatchNode(Values=("RightToeBase","Spine3")) {template="DISPATCH_RigVMDispatch_ArrayMake(in Values,out Array)"}
   n8 = RigVMDispatchNode(Index=1, Array=@Make_Array_1_1_1_1_1.Array) {template="DISPATCH_RigVMDispatch_ArrayGetAtIndex(in Array,in Index,out Element)"}
   RigVMFunction_ControlFlowBranch = &RigVMFunction_ControlFlowBranch(
    Condition=&RigUnit_ItemExists(Item=(Type=Bone,Name="None"), Item.Name=@n8.Element).Exists
    ExecuteContext=@RigVMFunction_Sequence_2.G
   )
   n9 = /Script/ControlRig.RigUnit_HierarchyGetParentsItemArray(Child=(Type=Bone,Name="None"), bDefaultParent=True, bIncludeChild=false, bReverse=False, Child.Name=@n8.Element) {is=struct RigVMUnitNode}
   RigVMFunction_MathBoolAnd_1_1_1_1_1 = RigVMAggregateNode(
    D=RigVMDispatchNode(
     B=Hips
     A=RigVMDispatchNode(Index=-1, Array=@n9.Parents) {template="DISPATCH_RigVMDispatch_ArrayGetAtIndex(in Array,in Index,out Element)"}.Element.Name
    ) {template="DISPATCH_RigVMDispatch_CoreEquals(in A,in B,out Result)"}.Result
    B=&RigUnit_ItemExists(Item=(Type=Bone,Name="None"), Item.Name=@n8.Element).Exists
    A=&RigUnit_ItemExists(
     Item=(Type=Bone,Name="None")
     Item.Name=RigVMDispatchNode(Index=0, Array=@Make_Array_1_1_1_1_1.Array) {template="DISPATCH_RigVMDispatch_ArrayGetAtIndex(in Array,in Index,out Element)"}.Element
    ).Exists
    C=&RigVMFunction_MathBoolNot(Value=RigVMVariableNode() {variable=SkeletonFound}.Value).Result
   )
   n10 = &RigVMFunction_ControlFlowBranch(ExecuteContext=@RigVMFunction_ControlFlowBranch.True, Condition=@RigVMFunction_MathBoolAnd_1_1_1_1_1.Result)
   RigVMVariableNode(Value=NewEnumerator7, ExecuteContext=@n10.True) {variable=Skeleton}
   > RigVMVariableNode(Value=true) {variable=SkeletonFound}
   From_String_2_1_1_1_2_1 = RigVMDispatchNode(String=Hips) {template="DISPATCH_RigDispatch_FromString(in String,out Result)"}
   > Set_Module_Metadata_3_1_1_2_1 = RigVMDispatchNode(Name=Pelvis Bone, NameSpace=Root, Value=(Type=Bone,Name="None"), Value.Name=@From_String_2_1_1_1_2_1.Result) {template="DISPATCH_RigDispatch_SetModuleMetadata(in Name,in NameSpace,in Value,out Success)"}
   From_String_1_1_1_1_2_1 = RigVMDispatchNode(String=Spine) {template="DISPATCH_RigDispatch_FromString(in String,out Result)"}
   > Set_Module_Metadata_1_1_1_1_2_1 = RigVMDispatchNode(Name=Spine Start Bone, NameSpace=Root, Value=(Type=Bone,Name="None"), Value.Name=@From_String_1_1_1_1_2_1.Result) {template="DISPATCH_RigDispatch_SetModuleMetadata(in Name,in NameSpace,in Value,out Success)"}
   From_String_4_1_1_2_1 = RigVMDispatchNode(String=Spine3) {template="DISPATCH_RigDispatch_FromString(in String,out Result)"}
   > Set_Module_Metadata_2_2_1_1_1_2_1 = RigVMDispatchNode(Name=Spine End Bone, NameSpace=Root, Value=(Type=Bone,Name="None"), Value.Name=@From_String_4_1_1_2_1.Result) {template="DISPATCH_RigDispatch_SetModuleMetadata(in Name,in NameSpace,in Value,out Success)"}
   From_String_3_2_1_1_2_1 = RigVMDispatchNode(String=LeftArm) {template="DISPATCH_RigDispatch_FromString(in String,out Result)"}
   > Set_Module_Metadata_2_1_2_1_1_2_1 = RigVMDispatchNode(Name=shoulder_l, NameSpace=Root, Value=(Type=Bone,Name="None"), Value.Name=@From_String_3_2_1_1_2_1.Result) {template="DISPATCH_RigDispatch_SetModuleMetadata(in Name,in NameSpace,in Value,out Success)"}
   From_String_3_1_4_1_1_2_1 = RigVMDispatchNode(String=RightArm) {template="DISPATCH_RigDispatch_FromString(in String,out Result)"}
   > Set_Module_Metadata_2_1_1_2_1_1_2_1 = RigVMDispatchNode(Name=shoulder_r, NameSpace=Root, Value=(Type=Bone,Name="None"), Value.Name=@From_String_3_1_4_1_1_2_1.Result) {template="DISPATCH_RigDispatch_SetModuleMetadata(in Name,in NameSpace,in Value,out Success)"}
   From_String_3_1_1_1_2_1_2_1 = RigVMDispatchNode(String=LeftUpLeg) {template="DISPATCH_RigDispatch_FromString(in String,out Result)"}
   > Set_Module_Metadata_2_1_1_1_2_2_1_2_1 = RigVMDispatchNode(Name=thigh_l, NameSpace=Root, Value=(Type=Bone,Name="None"), Value.Name=@From_String_3_1_1_1_2_1_2_1.Result) {template="DISPATCH_RigDispatch_SetModuleMetadata(in Name,in NameSpace,in Value,out Success)"}
   From_String_3_1_2_1_2_1_2_1 = RigVMDispatchNode(String=RightUpLeg) {template="DISPATCH_RigDispatch_FromString(in String,out Result)"}
   > Set_Module_Metadata_2_1_1_1_1_2_2_1_2_1 = RigVMDispatchNode(Name=thigh_r, NameSpace=Root, Value=(Type=Bone,Name="None"), Value.Name=@From_String_3_1_2_1_2_1_2_1.Result) {template="DISPATCH_RigDispatch_SetModuleMetadata(in Name,in NameSpace,in Value,out Success)"}
   From_String_3_1_1_1_1_1_1_2_1 = RigVMDispatchNode(String=LeftLeg) {template="DISPATCH_RigDispatch_FromString(in String,out Result)"}
   > Set_Module_Metadata_2_1_1_1_2_1_1_1_2_1 = RigVMDispatchNode(Name=knee_l, NameSpace=Root, Value=(Type=Bone,Name="None"), Value.Name=@From_String_3_1_1_1_1_1_1_2_1.Result) {template="DISPATCH_RigDispatch_SetModuleMetadata(in Name,in NameSpace,in Value,out Success)"}
   From_String_3_1_2_1_1_1_1_2_1 = RigVMDispatchNode(String=RightLeg) {template="DISPATCH_RigDispatch_FromString(in String,out Result)"}
   > Set_Module_Metadata_2_1_1_1_1_2_1_1_1_2_1 = RigVMDispatchNode(Name=knee_r, NameSpace=Root, Value=(Type=Bone,Name="None"), Value.Name=@From_String_3_1_2_1_1_1_1_2_1.Result) {template="DISPATCH_RigDispatch_SetModuleMetadata(in Name,in NameSpace,in Value,out Success)"}
   From_String_3_1_3_2_2_1_2_1 = RigVMDispatchNode(String=Neck) {template="DISPATCH_RigDispatch_FromString(in String,out Result)"}
   > Set_Module_Metadata_2_1_1_1_1_1_2_2_1_2_1 = RigVMDispatchNode(Name=neck, NameSpace=Root, Value=(Type=Bone,Name="None"), Value.Name=@From_String_3_1_3_2_2_1_2_1.Result) {template="DISPATCH_RigDispatch_SetModuleMetadata(in Name,in NameSpace,in Value,out Success)"}
   > Set_Module_Metadata_2_6_1_2_1_2_1 = RigVMDispatchNode(Name=Arm PV Position, NameSpace=Root, Value=10) {template="DISPATCH_RigDispatch_SetModuleMetadata(in Name,in NameSpace,in Value,out Success)"}
   > Set_Module_Metadata_2_3_3_1_1_1_2_1 = RigVMDispatchNode(Name=Leg PV Position, NameSpace=Root, Value=15) {template="DISPATCH_RigDispatch_SetModuleMetadata(in Name,in NameSpace,in Value,out Success)"}
   From_String_3_1_3_2_1_2_1_2_1 = RigVMDispatchNode(String=Head) {template="DISPATCH_RigDispatch_FromString(in String,out Result)"}
   > Set_Module_Metadata_2_1_1_1_1_1_2_1_2_1_2_1 = RigVMDispatchNode(Name=head, NameSpace=Root, Value=(Type=Bone,Name="None"), Value.Name=@From_String_3_1_3_2_1_2_1_2_1.Result) {template="DISPATCH_RigDispatch_SetModuleMetadata(in Name,in NameSpace,in Value,out Success)"}
   From_String_3_1_3_1_2_2_1_2_1 = RigVMDispatchNode(String=LeftFoot) {template="DISPATCH_RigDispatch_FromString(in String,out Result)"}
   > Set_Module_Metadata_2_1_1_1_1_1_1_2_2_1_2_1 = RigVMDispatchNode(Name=foot_l, NameSpace=Root, Value=(Type=Bone,Name="None"), Value.Name=@From_String_3_1_3_1_2_2_1_2_1.Result) {template="DISPATCH_RigDispatch_SetModuleMetadata(in Name,in NameSpace,in Value,out Success)"}
   From_String_3_1_3_1_1_1_3_1_2_1 = RigVMDispatchNode(String=RightFoot) {template="DISPATCH_RigDispatch_FromString(in String,out Result)"}
   > Set_Module_Metadata_2_1_1_1_1_1_1_1_1_2_1_2_1 = RigVMDispatchNode(Name=foot_r, NameSpace=Root, Value=(Type=Bone,Name="None"), Value.Name=@From_String_3_1_3_1_1_1_3_1_2_1.Result) {template="DISPATCH_RigDispatch_SetModuleMetadata(in Name,in NameSpace,in Value,out Success)"}
   From_String_3_1_3_1_2_1_4_1_2_1 = RigVMDispatchNode(String=LeftHand) {template="DISPATCH_RigDispatch_FromString(in String,out Result)"}
   > Set_Module_Metadata_2_1_1_1_1_1_1_2_1_4_1_2_1 = RigVMDispatchNode(Name=hand_l, NameSpace=Root, Value=(Type=Bone,Name="None"), Value.Name=@From_String_3_1_3_1_2_1_4_1_2_1.Result) {template="DISPATCH_RigDispatch_SetModuleMetadata(in Name,in NameSpace,in Value,out Success)"}
   From_String_3_1_3_1_1_1_1_5_1_2_1 = RigVMDispatchNode(String=RightHand) {template="DISPATCH_RigDispatch_FromString(in String,out Result)"}
   > Set_Module_Metadata_2_1_1_1_1_1_1_1_1_1_4_1_2_1 = RigVMDispatchNode(Name=hand_r, NameSpace=Root, Value=(Type=Bone,Name="None"), Value.Name=@From_String_3_1_3_1_1_1_1_5_1_2_1.Result) {template="DISPATCH_RigDispatch_SetModuleMetadata(in Name,in NameSpace,in Value,out Success)"}
   From_String_3_1_3_1_2_1_3_2_1_2_1 = RigVMDispatchNode(String=LeftForeArm) {template="DISPATCH_RigDispatch_FromString(in String,out Result)"}
   > Set_Module_Metadata_2_1_1_1_1_1_1_2_1_3_1_1_2_1 = RigVMDispatchNode(Name=elbow_l, NameSpace=Root, Value=(Type=Bone,Name="None"), Value.Name=@From_String_3_1_3_1_2_1_3_2_1_2_1.Result) {template="DISPATCH_RigDispatch_SetModuleMetadata(in Name,in NameSpace,in Value,out Success)"}
   From_String_3_1_3_1_1_1_1_3_2_1_2_1 = RigVMDispatchNode(String=RightForeArm) {template="DISPATCH_RigDispatch_FromString(in String,out Result)"}
   > Set_Module_Metadata_2_1_1_1_1_1_1_1_1_1_3_1_1_2_1 = RigVMDispatchNode(Name=elbow_r, NameSpace=Root, Value=(Type=Bone,Name="None"), Value.Name=@From_String_3_1_3_1_1_1_1_3_2_1_2_1.Result) {template="DISPATCH_RigDispatch_SetModuleMetadata(in Name,in NameSpace,in Value,out Success)"}
   > Set_Module_Metadata_2_3_2_1_3 = RigVMDispatchNode(Name=Primary Axis, NameSpace=Root, Value=(X=0,Y=0,Z=-1)) {template="DISPATCH_RigDispatch_SetModuleMetadata(in Name,in NameSpace,in Value,out Success)"}
   > Set_Module_Metadata_2_3_2_1_1_2 = RigVMDispatchNode(Name=Secondary Axis, NameSpace=Root, Value=(X=0,Y=-1,Z=0)) {template="DISPATCH_RigDispatch_SetModuleMetadata(in Name,in NameSpace,in Value,out Success)"}
   > Set_Module_Metadata_2_3_2_1_4 = RigVMDispatchNode(Name=Mirror Axis, NameSpace=Root, Value=(X=-1,Y=0,Z=0)) {template="DISPATCH_RigDispatch_SetModuleMetadata(in Name,in NameSpace,in Value,out Success)"}
   From_String_3_1_1_1_1_2 = RigVMDispatchNode(String=LeftToeBase) {template="DISPATCH_RigDispatch_FromString(in String,out Result)"}
   > Set_Module_Metadata_2_1_1_1_2_1_2 = RigVMDispatchNode(Name=toe_l, NameSpace=Root, Value=(Type=Bone,Name="None"), Value.Name=@From_String_3_1_1_1_1_2.Result) {template="DISPATCH_RigDispatch_SetModuleMetadata(in Name,in NameSpace,in Value,out Success)"}
   From_String_3_1_2_1_1_2 = RigVMDispatchNode(String=RightToeBase) {template="DISPATCH_RigDispatch_FromString(in String,out Result)"}
   > Set_Module_Metadata_2_1_1_1_1_2_1_2 = RigVMDispatchNode(Name=toe_r, NameSpace=Root, Value=(Type=Bone,Name="None"), Value.Name=@From_String_3_1_2_1_1_2.Result) {template="DISPATCH_RigDispatch_SetModuleMetadata(in Name,in NameSpace,in Value,out Success)"}
   > Set_Module_Metadata_6_1_2 = RigVMDispatchNode(Name=elbow_bone_offset, NameSpace=Root, Value=(X=0,Y=1,Z=0)) {template="DISPATCH_RigDispatch_SetModuleMetadata(in Name,in NameSpace,in Value,out Success)"}
   > Set_Module_Metadata_6_1_1_2 = RigVMDispatchNode(Name=knee_bone_offset, NameSpace=Root, Value=(X=0,Y=-1,Z=0)) {template="DISPATCH_RigDispatch_SetModuleMetadata(in Name,in NameSpace,in Value,out Success)"}
   Make_Array_1_1_1_1_1_1 = RigVMDispatchNode(Values=("LeftForeArm","RightForeArm")) {template="DISPATCH_RigVMDispatch_ArrayMake(in Values,out Array)"}
   n11 = RigVMDispatchNode(Index=1, Array=@Make_Array_1_1_1_1_1_1.Array) {template="DISPATCH_RigVMDispatch_ArrayGetAtIndex(in Array,in Index,out Element)"}
   n12 = &RigVMFunction_ControlFlowBranch(
    Condition=&RigUnit_ItemExists(Item=(Type=Bone,Name="None"), Item.Name=@n11.Element).Exists
    ExecuteContext=@RigVMFunction_Sequence_2.H
   )
   n13 = RigVMDispatchNode(
    Index=-1
    Array=&RigUnit_HierarchyGetParentsItemArray(Child=(Type=Bone,Name="None"), bDefaultParent=True, bIncludeChild=False, bReverse=False, Child.Name=@n11.Element).Parents
   ) {template="DISPATCH_RigVMDispatch_ArrayGetAtIndex(in Array,in Index,out Element)"}
   n14 = RigVMVariableNode() {variable=SkeletonFound}
   n15 = &RigVMFunction_ControlFlowBranch(
    ExecuteContext=@n12.True
    Condition=&RigVMFunction_MathBoolAnd(
     A=RigVMDispatchNode(B=Reference, A=@n13.Element.Name) {template="DISPATCH_RigVMDispatch_CoreEquals(in A,in B,out Result)"}.Result
     B=&RigVMFunction_MathBoolNot(Value=@n14.Value).Result
    ).Result
   )
   RigVMVariableNode(Value=NewEnumerator8, ExecuteContext=@n15.True) {variable=Skeleton}
   > RigVMVariableNode(Value=true) {variable=SkeletonFound}
   From_String_2_1_1_2 = RigVMDispatchNode(String=Hips) {template="DISPATCH_RigDispatch_FromString(in String,out Result)"}
   > Set_Module_Metadata_3_1_2 = RigVMDispatchNode(Name=Pelvis Bone, NameSpace=Root, Value=(Type=Bone,Name="None"), Value.Name=@From_String_2_1_1_2.Result) {template="DISPATCH_RigDispatch_SetModuleMetadata(in Name,in NameSpace,in Value,out Success)"}
   From_String_1_1_1_2 = RigVMDispatchNode(String=Spine) {template="DISPATCH_RigDispatch_FromString(in String,out Result)"}
   > Set_Module_Metadata_1_1_1_2 = RigVMDispatchNode(Name=Spine Start Bone, NameSpace=Root, Value=(Type=Bone,Name="None"), Value.Name=@From_String_1_1_1_2.Result) {template="DISPATCH_RigDispatch_SetModuleMetadata(in Name,in NameSpace,in Value,out Success)"}
   From_String_4_1_2 = RigVMDispatchNode(String=Spine2) {template="DISPATCH_RigDispatch_FromString(in String,out Result)"}
   > Set_Module_Metadata_2_2_1_1_2 = RigVMDispatchNode(Name=Spine End Bone, NameSpace=Root, Value=(Type=Bone,Name="None"), Value.Name=@From_String_4_1_2.Result) {template="DISPATCH_RigDispatch_SetModuleMetadata(in Name,in NameSpace,in Value,out Success)"}
   From_String_3_2_1_2 = RigVMDispatchNode(String=LeftArm) {template="DISPATCH_RigDispatch_FromString(in String,out Result)"}
   > Set_Module_Metadata_2_1_2_1_2 = RigVMDispatchNode(Name=shoulder_l, NameSpace=Root, Value=(Type=Bone,Name="None"), Value.Name=@From_String_3_2_1_2.Result) {template="DISPATCH_RigDispatch_SetModuleMetadata(in Name,in NameSpace,in Value,out Success)"}
   From_String_3_1_4_1_2 = RigVMDispatchNode(String=RightArm) {template="DISPATCH_RigDispatch_FromString(in String,out Result)"}
   > Set_Module_Metadata_2_1_1_2_1_2 = RigVMDispatchNode(Name=shoulder_r, NameSpace=Root, Value=(Type=Bone,Name="None"), Value.Name=@From_String_3_1_4_1_2.Result) {template="DISPATCH_RigDispatch_SetModuleMetadata(in Name,in NameSpace,in Value,out Success)"}
   From_String_3_1_1_1_2_2 = RigVMDispatchNode(String=LeftUpLeg) {template="DISPATCH_RigDispatch_FromString(in String,out Result)"}
   > Set_Module_Metadata_2_1_1_1_2_2_2 = RigVMDispatchNode(Name=thigh_l, NameSpace=Root, Value=(Type=Bone,Name="None"), Value.Name=@From_String_3_1_1_1_2_2.Result) {template="DISPATCH_RigDispatch_SetModuleMetadata(in Name,in NameSpace,in Value,out Success)"}
   From_String_3_1_2_1_2_2 = RigVMDispatchNode(String=RightUpLeg) {template="DISPATCH_RigDispatch_FromString(in String,out Result)"}
   > Set_Module_Metadata_2_1_1_1_1_2_2_2 = RigVMDispatchNode(Name=thigh_r, NameSpace=Root, Value=(Type=Bone,Name="None"), Value.Name=@From_String_3_1_2_1_2_2.Result) {template="DISPATCH_RigDispatch_SetModuleMetadata(in Name,in NameSpace,in Value,out Success)"}
   From_String_3_1_1_1_1_1_2 = RigVMDispatchNode(String=LeftLeg) {template="DISPATCH_RigDispatch_FromString(in String,out Result)"}
   > Set_Module_Metadata_2_1_1_1_2_1_1_2 = RigVMDispatchNode(Name=knee_l, NameSpace=Root, Value=(Type=Bone,Name="None"), Value.Name=@From_String_3_1_1_1_1_1_2.Result) {template="DISPATCH_RigDispatch_SetModuleMetadata(in Name,in NameSpace,in Value,out Success)"}
   From_String_3_1_2_1_1_1_2 = RigVMDispatchNode(String=RightLeg) {template="DISPATCH_RigDispatch_FromString(in String,out Result)"}
   > Set_Module_Metadata_2_1_1_1_1_2_1_1_2 = RigVMDispatchNode(Name=knee_r, NameSpace=Root, Value=(Type=Bone,Name="None"), Value.Name=@From_String_3_1_2_1_1_1_2.Result) {template="DISPATCH_RigDispatch_SetModuleMetadata(in Name,in NameSpace,in Value,out Success)"}
   From_String_3_1_3_2_2_2 = RigVMDispatchNode(String=Neck) {template="DISPATCH_RigDispatch_FromString(in String,out Result)"}
   > Set_Module_Metadata_2_1_1_1_1_1_2_2_2 = RigVMDispatchNode(Name=neck, NameSpace=Root, Value=(Type=Bone,Name="None"), Value.Name=@From_String_3_1_3_2_2_2.Result) {template="DISPATCH_RigDispatch_SetModuleMetadata(in Name,in NameSpace,in Value,out Success)"}
   > Set_Module_Metadata_2_6_1_2_3 = RigVMDispatchNode(Name=Arm PV Position, NameSpace=Root, Value=250) {template="DISPATCH_RigDispatch_SetModuleMetadata(in Name,in NameSpace,in Value,out Success)"}
   > Set_Module_Metadata_2_3_3_1_1_3 = RigVMDispatchNode(Name=Leg PV Position, NameSpace=Root, Value=15) {template="DISPATCH_RigDispatch_SetModuleMetadata(in Name,in NameSpace,in Value,out Success)"}
   From_String_3_1_3_2_1_2_2 = RigVMDispatchNode(String=Head) {template="DISPATCH_RigDispatch_FromString(in String,out Result)"}
   > Set_Module_Metadata_2_1_1_1_1_1_2_1_2_2 = RigVMDispatchNode(Name=head, NameSpace=Root, Value=(Type=Bone,Name="None"), Value.Name=@From_String_3_1_3_2_1_2_2.Result) {template="DISPATCH_RigDispatch_SetModuleMetadata(in Name,in NameSpace,in Value,out Success)"}
   From_String_3_1_3_1_2_2_2 = RigVMDispatchNode(String=LeftFoot) {template="DISPATCH_RigDispatch_FromString(in String,out Result)"}
   > Set_Module_Metadata_2_1_1_1_1_1_1_2_2_2 = RigVMDispatchNode(Name=foot_l, NameSpace=Root, Value=(Type=Bone,Name="None"), Value.Name=@From_String_3_1_3_1_2_2_2.Result) {template="DISPATCH_RigDispatch_SetModuleMetadata(in Name,in NameSpace,in Value,out Success)"}
   From_String_3_1_3_1_1_1_3_2 = RigVMDispatchNode(String=RightFoot) {template="DISPATCH_RigDispatch_FromString(in String,out Result)"}
   > Set_Module_Metadata_2_1_1_1_1_1_1_1_1_2_2 = RigVMDispatchNode(Name=foot_r, NameSpace=Root, Value=(Type=Bone,Name="None"), Value.Name=@From_String_3_1_3_1_1_1_3_2.Result) {template="DISPATCH_RigDispatch_SetModuleMetadata(in Name,in NameSpace,in Value,out Success)"}
   From_String_3_1_3_1_2_1_4_2 = RigVMDispatchNode(String=LeftHand) {template="DISPATCH_RigDispatch_FromString(in String,out Result)"}
   > Set_Module_Metadata_2_1_1_1_1_1_1_2_1_4_2 = RigVMDispatchNode(Name=hand_l, NameSpace=Root, Value=(Type=Bone,Name="None"), Value.Name=@From_String_3_1_3_1_2_1_4_2.Result) {template="DISPATCH_RigDispatch_SetModuleMetadata(in Name,in NameSpace,in Value,out Success)"}
   From_String_3_1_3_1_1_1_1_5_2 = RigVMDispatchNode(String=RightHand) {template="DISPATCH_RigDispatch_FromString(in String,out Result)"}
   > Set_Module_Metadata_2_1_1_1_1_1_1_1_1_1_4_2 = RigVMDispatchNode(Name=hand_r, NameSpace=Root, Value=(Type=Bone,Name="None"), Value.Name=@From_String_3_1_3_1_1_1_1_5_2.Result) {template="DISPATCH_RigDispatch_SetModuleMetadata(in Name,in NameSpace,in Value,out Success)"}
   From_String_3_1_3_1_2_1_3_2_2 = RigVMDispatchNode(String=LeftForeArm) {template="DISPATCH_RigDispatch_FromString(in String,out Result)"}
   > Set_Module_Metadata_2_1_1_1_1_1_1_2_1_3_1_2 = RigVMDispatchNode(Name=elbow_l, NameSpace=Root, Value=(Type=Bone,Name="None"), Value.Name=@From_String_3_1_3_1_2_1_3_2_2.Result) {template="DISPATCH_RigDispatch_SetModuleMetadata(in Name,in NameSpace,in Value,out Success)"}
   From_String_3_1_3_1_1_1_1_3_2_2 = RigVMDispatchNode(String=RightForeARm) {template="DISPATCH_RigDispatch_FromString(in String,out Result)"}
   > Set_Module_Metadata_2_1_1_1_1_1_1_1_1_1_3_1_2 = RigVMDispatchNode(Name=elbow_r, NameSpace=Root, Value=(Type=Bone,Name="None"), Value.Name=@From_String_3_1_3_1_1_1_1_3_2_2.Result) {template="DISPATCH_RigDispatch_SetModuleMetadata(in Name,in NameSpace,in Value,out Success)"}
   > Set_Module_Metadata_6_2 = RigVMDispatchNode(Name=elbow_bone_offset, NameSpace=Self, Value=(X=0,Y=1,Z=1)) {template="DISPATCH_RigDispatch_SetModuleMetadata(in Name,in NameSpace,in Value,out Success)"}
   From_String_3_1_1_1_1_2_1_2_1_2 = RigVMDispatchNode(String=LeftToeBase) {template="DISPATCH_RigDispatch_FromString(in String,out Result)"}
   > Set_Module_Metadata_2_1_1_1_2_1_2_1_2_1_2 = RigVMDispatchNode(Name=toe_l, NameSpace=Root, Value=(Type=Bone,Name="None"), Value.Name=@From_String_3_1_1_1_1_2_1_2_1_2.Result) {template="DISPATCH_RigDispatch_SetModuleMetadata(in Name,in NameSpace,in Value,out Success)"}
   From_String_3_1_2_1_1_2_1_2_1_2 = RigVMDispatchNode(String=RightToeBase) {template="DISPATCH_RigDispatch_FromString(in String,out Result)"}
   > Set_Module_Metadata_2_1_1_1_1_2_1_2_1_2_1_2 = RigVMDispatchNode(Name=toe_r, NameSpace=Root, Value=(Type=Bone,Name="None"), Value.Name=@From_String_3_1_2_1_1_2_1_2_1_2.Result) {template="DISPATCH_RigDispatch_SetModuleMetadata(in Name,in NameSpace,in Value,out Success)"}
   > Set_Module_Metadata_6_1_3 = RigVMDispatchNode(Name=elbow_bone_offset, NameSpace=Root, Value=(X=0,Y=-0.1,Z=0)) {template="DISPATCH_RigDispatch_SetModuleMetadata(in Name,in NameSpace,in Value,out Success)"}
   > Set_Module_Metadata_6_1_1_3 = RigVMDispatchNode(Name=knee_bone_offset, NameSpace=Root, Value=(X=0,Y=0,Z=0)) {template="DISPATCH_RigDispatch_SetModuleMetadata(in Name,in NameSpace,in Value,out Success)"}
   > Set_Module_Metadata_2_3_2_1_4_1_2_1 = RigVMDispatchNode(Name=Mirror Axis, NameSpace=Root, Value=(X=1,Y=0,Z=0)) {template="DISPATCH_RigDispatch_SetModuleMetadata(in Name,in NameSpace,in Value,out Success)"}
   > Set_Module_Metadata_6_1_1_1_1 = RigVMDispatchNode(Name=elbow_bone_offset_globalSpace, NameSpace=Root, Value=true) {template="DISPATCH_RigDispatch_SetModuleMetadata(in Name,in NameSpace,in Value,out Success)"}
   Make_Array_1_1_1_1_1_1_1_1_1_1_1 = RigVMDispatchNode(Values=("spine_03")) {template="DISPATCH_RigVMDispatch_ArrayMake(in Values,out Array)"}
   n16 = &RigUnit_CollectionChildrenArray(
    Parent=(Type=Bone,Name="None")
    TypeToSearch=Bone
    bDefaultChildren=True
    bIncludeParent=False
    bRecursive=False
    Parent.Name=RigVMDispatchNode(Index=0, Array=@Make_Array_1_1_1_1_1_1_1_1_1_1_1.Array) {template="DISPATCH_RigVMDispatch_ArrayGetAtIndex(in Array,in Index,out Element)"}.Element
   )
   Make_Array_1_1_1_1_1_1_1_1_2 = RigVMDispatchNode(Values=("spine_01","spine_03")) {template="DISPATCH_RigVMDispatch_ArrayMake(in Values,out Array)"}
   Make_Array_1_1_1_1_1_1_1_1_1_1 = RigVMDispatchNode(Values=("spine_04")) {template="DISPATCH_RigVMDispatch_ArrayMake(in Values,out Array)"}
   n17 = RigVMDispatchNode(Index=0, Array=@Make_Array_1_1_1_1_1_1_1_1_1_1.Array) {template="DISPATCH_RigVMDispatch_ArrayGetAtIndex(in Array,in Index,out Element)"}
   RigVMFunction_MathBoolAnd_1_1_1_1_1_1_1_1_1 = RigVMAggregateNode(
    D=RigVMDispatchNode(B=3, A=RigVMDispatchNode(Array=@n16.Items) {template="DISPATCH_RigVMDispatch_ArrayGetNum(in Array,out Num)"}.Num) {template="DISPATCH_RigVMDispatch_CoreEquals(in A,in B,out Result)"}.Result
    B=&RigUnit_ItemExists(
     Item=(Type=Bone,Name="None")
     Item.Name=RigVMDispatchNode(Index=1, Array=@Make_Array_1_1_1_1_1_1_1_1_2.Array) {template="DISPATCH_RigVMDispatch_ArrayGetAtIndex(in Array,in Index,out Element)"}.Element
    ).Exists
    A=&RigUnit_ItemExists(
     Item=(Type=Bone,Name="None")
     Item.Name=RigVMDispatchNode(Index=0, Array=@Make_Array_1_1_1_1_1_1_1_1_2.Array) {template="DISPATCH_RigVMDispatch_ArrayGetAtIndex(in Array,in Index,out Element)"}.Element
    ).Exists
    E=&RigVMFunction_MathBoolNot(Value=RigVMVariableNode() {variable=SkeletonFound}.Value).Result
    C=&RigVMFunction_MathBoolNot(Value=&RigUnit_ItemExists(Item=(Type=Bone,Name="None"), Item.Name=@n17.Element).Exists).Result
   )
   n18 = &RigVMFunction_ControlFlowBranch(Condition=@RigVMFunction_MathBoolAnd_1_1_1_1_1_1_1_1_1.Result, ExecuteContext=@RigVMFunction_Sequence_2.I)
   RigVMVariableNode(Value=NewEnumerator1, ExecuteContext=@n18.True) {variable=Skeleton}
   > RigVMVariableNode(Value=true) {variable=SkeletonFound}
   From_String_2_1_1_1_2 = RigVMDispatchNode(String=pelvis) {template="DISPATCH_RigDispatch_FromString(in String,out Result)"}
   > Set_Module_Metadata_3_1_1_2 = RigVMDispatchNode(Name=Pelvis Bone, NameSpace=Root, Value=(Type=Bone,Name="None"), Value.Name=@From_String_2_1_1_1_2.Result) {template="DISPATCH_RigDispatch_SetModuleMetadata(in Name,in NameSpace,in Value,out Success)"}
   From_String_1_1_1_1_2 = RigVMDispatchNode(String=spine_01) {template="DISPATCH_RigDispatch_FromString(in String,out Result)"}
   > Set_Module_Metadata_1_1_1_1_2 = RigVMDispatchNode(Name=Spine Start Bone, NameSpace=Root, Value=(Type=Bone,Name="None"), Value.Name=@From_String_1_1_1_1_2.Result) {template="DISPATCH_RigDispatch_SetModuleMetadata(in Name,in NameSpace,in Value,out Success)"}
   From_String_4_1_1_2 = RigVMDispatchNode(String=spine_02) {template="DISPATCH_RigDispatch_FromString(in String,out Result)"}
   > Set_Module_Metadata_2_2_1_1_1_2 = RigVMDispatchNode(Name=Spine End Bone, NameSpace=Root, Value=(Type=Bone,Name="None"), Value.Name=@From_String_4_1_1_2.Result) {template="DISPATCH_RigDispatch_SetModuleMetadata(in Name,in NameSpace,in Value,out Success)"}
   From_String_3_2_1_1_2 = RigVMDispatchNode(String=upperarm_l) {template="DISPATCH_RigDispatch_FromString(in String,out Result)"}
   > Set_Module_Metadata_2_1_2_1_1_2 = RigVMDispatchNode(Name=shoulder_l, NameSpace=Root, Value=(Type=Bone,Name="None"), Value.Name=@From_String_3_2_1_1_2.Result) {template="DISPATCH_RigDispatch_SetModuleMetadata(in Name,in NameSpace,in Value,out Success)"}
   From_String_3_1_4_1_1_2 = RigVMDispatchNode(String=upperarm_r) {template="DISPATCH_RigDispatch_FromString(in String,out Result)"}
   > Set_Module_Metadata_2_1_1_2_1_1_2 = RigVMDispatchNode(Name=shoulder_r, NameSpace=Root, Value=(Type=Bone,Name="None"), Value.Name=@From_String_3_1_4_1_1_2.Result) {template="DISPATCH_RigDispatch_SetModuleMetadata(in Name,in NameSpace,in Value,out Success)"}
   From_String_3_1_1_1_2_1_2 = RigVMDispatchNode(String=thigh_l) {template="DISPATCH_RigDispatch_FromString(in String,out Result)"}
   > Set_Module_Metadata_2_1_1_1_2_2_1_2 = RigVMDispatchNode(Name=thigh_l, NameSpace=Root, Value=(Type=Bone,Name="None"), Value.Name=@From_String_3_1_1_1_2_1_2.Result) {template="DISPATCH_RigDispatch_SetModuleMetadata(in Name,in NameSpace,in Value,out Success)"}
   From_String_3_1_2_1_2_1_2 = RigVMDispatchNode(String=thigh_r) {template="DISPATCH_RigDispatch_FromString(in String,out Result)"}
   > Set_Module_Metadata_2_1_1_1_1_2_2_1_2 = RigVMDispatchNode(Name=thigh_r, NameSpace=Root, Value=(Type=Bone,Name="None"), Value.Name=@From_String_3_1_2_1_2_1_2.Result) {template="DISPATCH_RigDispatch_SetModuleMetadata(in Name,in NameSpace,in Value,out Success)"}
   From_String_3_1_1_1_1_1_1_2 = RigVMDispatchNode(String=calf_l) {template="DISPATCH_RigDispatch_FromString(in String,out Result)"}
   > Set_Module_Metadata_2_1_1_1_2_1_1_1_2 = RigVMDispatchNode(Name=knee_l, NameSpace=Root, Value=(Type=Bone,Name="None"), Value.Name=@From_String_3_1_1_1_1_1_1_2.Result) {template="DISPATCH_RigDispatch_SetModuleMetadata(in Name,in NameSpace,in Value,out Success)"}
   From_String_3_1_2_1_1_1_1_2 = RigVMDispatchNode(String=calf_r) {template="DISPATCH_RigDispatch_FromString(in String,out Result)"}
   > Set_Module_Metadata_2_1_1_1_1_2_1_1_1_2 = RigVMDispatchNode(Name=knee_r, NameSpace=Root, Value=(Type=Bone,Name="None"), Value.Name=@From_String_3_1_2_1_1_1_1_2.Result) {template="DISPATCH_RigDispatch_SetModuleMetadata(in Name,in NameSpace,in Value,out Success)"}
   From_String_3_1_3_2_2_1_2 = RigVMDispatchNode(String=neck_01) {template="DISPATCH_RigDispatch_FromString(in String,out Result)"}
   > Set_Module_Metadata_2_1_1_1_1_1_2_2_1_2 = RigVMDispatchNode(Name=neck, NameSpace=Root, Value=(Type=Bone,Name="None"), Value.Name=@From_String_3_1_3_2_2_1_2.Result) {template="DISPATCH_RigDispatch_SetModuleMetadata(in Name,in NameSpace,in Value,out Success)"}
   > Set_Module_Metadata_2_6_1_2_1_2 = RigVMDispatchNode(Name=Arm PV Position, NameSpace=Root, Value=10) {template="DISPATCH_RigDispatch_SetModuleMetadata(in Name,in NameSpace,in Value,out Success)"}
   > Set_Module_Metadata_2_3_3_1_1_1_2 = RigVMDispatchNode(Name=Leg PV Position, NameSpace=Root, Value=15) {template="DISPATCH_RigDispatch_SetModuleMetadata(in Name,in NameSpace,in Value,out Success)"}
   From_String_3_1_3_2_1_2_1_2 = RigVMDispatchNode(String=Head) {template="DISPATCH_RigDispatch_FromString(in String,out Result)"}
   > Set_Module_Metadata_2_1_1_1_1_1_2_1_2_1_2 = RigVMDispatchNode(Name=head, NameSpace=Root, Value=(Type=Bone,Name="None"), Value.Name=@From_String_3_1_3_2_1_2_1_2.Result) {template="DISPATCH_RigDispatch_SetModuleMetadata(in Name,in NameSpace,in Value,out Success)"}
   From_String_3_1_3_1_2_2_1_2 = RigVMDispatchNode(String=foot_l) {template="DISPATCH_RigDispatch_FromString(in String,out Result)"}
   > Set_Module_Metadata_2_1_1_1_1_1_1_2_2_1_2 = RigVMDispatchNode(Name=foot_l, NameSpace=Root, Value=(Type=Bone,Name="None"), Value.Name=@From_String_3_1_3_1_2_2_1_2.Result) {template="DISPATCH_RigDispatch_SetModuleMetadata(in Name,in NameSpace,in Value,out Success)"}
   From_String_3_1_3_1_1_1_3_1_2 = RigVMDispatchNode(String=foot_r) {template="DISPATCH_RigDispatch_FromString(in String,out Result)"}
   > Set_Module_Metadata_2_1_1_1_1_1_1_1_1_2_1_2 = RigVMDispatchNode(Name=foot_r, NameSpace=Root, Value=(Type=Bone,Name="None"), Value.Name=@From_String_3_1_3_1_1_1_3_1_2.Result) {template="DISPATCH_RigDispatch_SetModuleMetadata(in Name,in NameSpace,in Value,out Success)"}
   From_String_3_1_3_1_2_1_4_1_2 = RigVMDispatchNode(String=hand_l) {template="DISPATCH_RigDispatch_FromString(in String,out Result)"}
   > Set_Module_Metadata_2_1_1_1_1_1_1_2_1_4_1_2 = RigVMDispatchNode(Name=hand_l, NameSpace=Root, Value=(Type=Bone,Name="None"), Value.Name=@From_String_3_1_3_1_2_1_4_1_2.Result) {template="DISPATCH_RigDispatch_SetModuleMetadata(in Name,in NameSpace,in Value,out Success)"}
   From_String_3_1_3_1_1_1_1_5_1_2 = RigVMDispatchNode(String=hand_r) {template="DISPATCH_RigDispatch_FromString(in String,out Result)"}
   > Set_Module_Metadata_2_1_1_1_1_1_1_1_1_1_4_1_2 = RigVMDispatchNode(Name=hand_r, NameSpace=Root, Value=(Type=Bone,Name="None"), Value.Name=@From_String_3_1_3_1_1_1_1_5_1_2.Result) {template="DISPATCH_RigDispatch_SetModuleMetadata(in Name,in NameSpace,in Value,out Success)"}
   From_String_3_1_3_1_2_1_3_2_1_2 = RigVMDispatchNode(String=lowerarm_l) {template="DISPATCH_RigDispatch_FromString(in String,out Result)"}
   > Set_Module_Metadata_2_1_1_1_1_1_1_2_1_3_1_1_2 = RigVMDispatchNode(Name=elbow_l, NameSpace=Root, Value=(Type=Bone,Name="None"), Value.Name=@From_String_3_1_3_1_2_1_3_2_1_2.Result) {template="DISPATCH_RigDispatch_SetModuleMetadata(in Name,in NameSpace,in Value,out Success)"}
   From_String_3_1_3_1_1_1_1_3_2_1_2 = RigVMDispatchNode(String=lowerarm_r) {template="DISPATCH_RigDispatch_FromString(in String,out Result)"}
   > Set_Module_Metadata_2_1_1_1_1_1_1_1_1_1_3_1_1_2 = RigVMDispatchNode(Name=elbow_r, NameSpace=Root, Value=(Type=Bone,Name="None"), Value.Name=@From_String_3_1_3_1_1_1_1_3_2_1_2.Result) {template="DISPATCH_RigDispatch_SetModuleMetadata(in Name,in NameSpace,in Value,out Success)"}
   From_String_3_1_1_1_1_2_1_2_1_1 = RigVMDispatchNode(String=ball_l) {template="DISPATCH_RigDispatch_FromString(in String,out Result)"}
   > Set_Module_Metadata_2_1_1_1_2_1_2_1_2_1_1 = RigVMDispatchNode(Name=toe_l, NameSpace=Root, Value=(Type=Bone,Name="None"), Value.Name=@From_String_3_1_1_1_1_2_1_2_1_1.Result) {template="DISPATCH_RigDispatch_SetModuleMetadata(in Name,in NameSpace,in Value,out Success)"}
   From_String_3_1_2_1_1_2_1_2_1_1 = RigVMDispatchNode(String=ball_r) {template="DISPATCH_RigDispatch_FromString(in String,out Result)"}
   > Set_Module_Metadata_2_1_1_1_1_2_1_2_1_2_1_1 = RigVMDispatchNode(Name=toe_r, NameSpace=Root, Value=(Type=Bone,Name="None"), Value.Name=@From_String_3_1_2_1_1_2_1_2_1_1.Result) {template="DISPATCH_RigDispatch_SetModuleMetadata(in Name,in NameSpace,in Value,out Success)"}
   Make_Array_1_1_1_1_1_1_1_1_1_1_1_1 = RigVMDispatchNode(Values=("spine_03")) {template="DISPATCH_RigVMDispatch_ArrayMake(in Values,out Array)"}
   n19 = RigVMDispatchNode(Index=0, Array=@Make_Array_1_1_1_1_1_1_1_1_1_1_1_1.Array) {template="DISPATCH_RigVMDispatch_ArrayGetAtIndex(in Array,in Index,out Element)"}
   Greater = /Script/RigVM.RigVMFunction_MathIntGreater(
    B=3
    A=RigVMDispatchNode(Array=&RigUnit_CollectionChildrenArray(Parent=(Type=Bone,Name="None"), TypeToSearch=Bone, bDefaultChildren=True, bIncludeParent=False, bRecursive=False, Parent.Name=@n19.Element).Items) {template="DISPATCH_RigVMDispatch_ArrayGetNum(in Array,out Num)"}.Num
   ) {is=struct RigVMUnitNode, template="Greater::Execute(in A,in B,out Result)"}
   Make_Array_1_1_1_1_1_1_1_1 = RigVMDispatchNode(Values=("spine_01","spine_03")) {template="DISPATCH_RigVMDispatch_ArrayMake(in Values,out Array)"}
   Make_Array_1_1_1_1_1_1_1_1_1 = RigVMDispatchNode(Values=("spine_04")) {template="DISPATCH_RigVMDispatch_ArrayMake(in Values,out Array)"}
   n20 = RigVMDispatchNode(Index=0, Array=@Make_Array_1_1_1_1_1_1_1_1_1.Array) {template="DISPATCH_RigVMDispatch_ArrayGetAtIndex(in Array,in Index,out Element)"}
   RigVMFunction_MathBoolAnd_1_1_1_1_1_1_1_1 = RigVMAggregateNode(
    D=@Greater.Result
    B=&RigUnit_ItemExists(
     Item=(Type=Bone,Name="None")
     Item.Name=RigVMDispatchNode(Index=1, Array=@Make_Array_1_1_1_1_1_1_1_1.Array) {template="DISPATCH_RigVMDispatch_ArrayGetAtIndex(in Array,in Index,out Element)"}.Element
    ).Exists
    A=&RigUnit_ItemExists(
     Item=(Type=Bone,Name="None")
     Item.Name=RigVMDispatchNode(Index=0, Array=@Make_Array_1_1_1_1_1_1_1_1.Array) {template="DISPATCH_RigVMDispatch_ArrayGetAtIndex(in Array,in Index,out Element)"}.Element
    ).Exists
    E=&RigVMFunction_MathBoolNot(Value=RigVMVariableNode() {variable=SkeletonFound}.Value).Result
    C=&RigVMFunction_MathBoolNot(Value=&RigUnit_ItemExists(Item=(Type=Bone,Name="None"), Item.Name=@n20.Element).Exists).Result
   )
   n21 = &RigVMFunction_ControlFlowBranch(Condition=@RigVMFunction_MathBoolAnd_1_1_1_1_1_1_1_1.Result, ExecuteContext=@RigVMFunction_Sequence_2.J)
   RigVMVariableNode(Value=NewEnumerator9, ExecuteContext=@n21.True) {variable=Skeleton}
   > RigVMVariableNode(Value=true) {variable=SkeletonFound}
   From_String_2_1_1_1_1 = RigVMDispatchNode(String=pelvis) {template="DISPATCH_RigDispatch_FromString(in String,out Result)"}
   > Set_Module_Metadata_3_1_1_1 = RigVMDispatchNode(Name=Pelvis Bone, NameSpace=Root, Value=(Type=Bone,Name="None"), Value.Name=@From_String_2_1_1_1_1.Result) {template="DISPATCH_RigDispatch_SetModuleMetadata(in Name,in NameSpace,in Value,out Success)"}
   From_String_1_1_1_1_1 = RigVMDispatchNode(String=spine_01) {template="DISPATCH_RigDispatch_FromString(in String,out Result)"}
   > Set_Module_Metadata_1_1_1_1_1 = RigVMDispatchNode(Name=Spine Start Bone, NameSpace=Root, Value=(Type=Bone,Name="None"), Value.Name=@From_String_1_1_1_1_1.Result) {template="DISPATCH_RigDispatch_SetModuleMetadata(in Name,in NameSpace,in Value,out Success)"}
   From_String_4_1_1_1 = RigVMDispatchNode(String=spine_02) {template="DISPATCH_RigDispatch_FromString(in String,out Result)"}
   > Set_Module_Metadata_2_2_1_1_1_1 = RigVMDispatchNode(Name=Spine End Bone, NameSpace=Root, Value=(Type=Bone,Name="None"), Value.Name=@From_String_4_1_1_1.Result) {template="DISPATCH_RigDispatch_SetModuleMetadata(in Name,in NameSpace,in Value,out Success)"}
   From_String_3_2_1_1_1 = RigVMDispatchNode(String=upperarm_l) {template="DISPATCH_RigDispatch_FromString(in String,out Result)"}
   > Set_Module_Metadata_2_1_2_1_1_1 = RigVMDispatchNode(Name=shoulder_l, NameSpace=Root, Value=(Type=Bone,Name="None"), Value.Name=@From_String_3_2_1_1_1.Result) {template="DISPATCH_RigDispatch_SetModuleMetadata(in Name,in NameSpace,in Value,out Success)"}
   From_String_3_1_4_1_1_1 = RigVMDispatchNode(String=upperarm_r) {template="DISPATCH_RigDispatch_FromString(in String,out Result)"}
   > Set_Module_Metadata_2_1_1_2_1_1_1 = RigVMDispatchNode(Name=shoulder_r, NameSpace=Root, Value=(Type=Bone,Name="None"), Value.Name=@From_String_3_1_4_1_1_1.Result) {template="DISPATCH_RigDispatch_SetModuleMetadata(in Name,in NameSpace,in Value,out Success)"}
   From_String_3_1_1_1_2_1_1 = RigVMDispatchNode(String=thigh_l) {template="DISPATCH_RigDispatch_FromString(in String,out Result)"}
   > Set_Module_Metadata_2_1_1_1_2_2_1_1 = RigVMDispatchNode(Name=thigh_l, NameSpace=Root, Value=(Type=Bone,Name="None"), Value.Name=@From_String_3_1_1_1_2_1_1.Result) {template="DISPATCH_RigDispatch_SetModuleMetadata(in Name,in NameSpace,in Value,out Success)"}
   From_String_3_1_2_1_2_1_1 = RigVMDispatchNode(String=thigh_r) {template="DISPATCH_RigDispatch_FromString(in String,out Result)"}
   > Set_Module_Metadata_2_1_1_1_1_2_2_1_1 = RigVMDispatchNode(Name=thigh_r, NameSpace=Root, Value=(Type=Bone,Name="None"), Value.Name=@From_String_3_1_2_1_2_1_1.Result) {template="DISPATCH_RigDispatch_SetModuleMetadata(in Name,in NameSpace,in Value,out Success)"}
   From_String_3_1_1_1_1_1_1_1 = RigVMDispatchNode(String=calf_l) {template="DISPATCH_RigDispatch_FromString(in String,out Result)"}
   > Set_Module_Metadata_2_1_1_1_2_1_1_1_1 = RigVMDispatchNode(Name=knee_l, NameSpace=Root, Value=(Type=Bone,Name="None"), Value.Name=@From_String_3_1_1_1_1_1_1_1.Result) {template="DISPATCH_RigDispatch_SetModuleMetadata(in Name,in NameSpace,in Value,out Success)"}
   From_String_3_1_2_1_1_1_1_1 = RigVMDispatchNode(String=calf_r) {template="DISPATCH_RigDispatch_FromString(in String,out Result)"}
   > Set_Module_Metadata_2_1_1_1_1_2_1_1_1_1 = RigVMDispatchNode(Name=knee_r, NameSpace=Root, Value=(Type=Bone,Name="None"), Value.Name=@From_String_3_1_2_1_1_1_1_1.Result) {template="DISPATCH_RigDispatch_SetModuleMetadata(in Name,in NameSpace,in Value,out Success)"}
   From_String_3_1_3_2_2_1_1 = RigVMDispatchNode(String=neck_01) {template="DISPATCH_RigDispatch_FromString(in String,out Result)"}
   > Set_Module_Metadata_2_1_1_1_1_1_2_2_1_1 = RigVMDispatchNode(Name=neck, NameSpace=Root, Value=(Type=Bone,Name="None"), Value.Name=@From_String_3_1_3_2_2_1_1.Result) {template="DISPATCH_RigDispatch_SetModuleMetadata(in Name,in NameSpace,in Value,out Success)"}
   > Set_Module_Metadata_2_6_1_2_1_1 = RigVMDispatchNode(Name=Arm PV Position, NameSpace=Root, Value=10) {template="DISPATCH_RigDispatch_SetModuleMetadata(in Name,in NameSpace,in Value,out Success)"}
   > Set_Module_Metadata_2_3_3_1_1_1_1 = RigVMDispatchNode(Name=Leg PV Position, NameSpace=Root, Value=15) {template="DISPATCH_RigDispatch_SetModuleMetadata(in Name,in NameSpace,in Value,out Success)"}
   From_String_3_1_3_2_1_2_1_1 = RigVMDispatchNode(String=Head) {template="DISPATCH_RigDispatch_FromString(in String,out Result)"}
   > Set_Module_Metadata_2_1_1_1_1_1_2_1_2_1_1 = RigVMDispatchNode(Name=head, NameSpace=Root, Value=(Type=Bone,Name="None"), Value.Name=@From_String_3_1_3_2_1_2_1_1.Result) {template="DISPATCH_RigDispatch_SetModuleMetadata(in Name,in NameSpace,in Value,out Success)"}
   From_String_3_1_3_1_2_2_1_1 = RigVMDispatchNode(String=foot_l) {template="DISPATCH_RigDispatch_FromString(in String,out Result)"}
   > Set_Module_Metadata_2_1_1_1_1_1_1_2_2_1_1 = RigVMDispatchNode(Name=foot_l, NameSpace=Root, Value=(Type=Bone,Name="None"), Value.Name=@From_String_3_1_3_1_2_2_1_1.Result) {template="DISPATCH_RigDispatch_SetModuleMetadata(in Name,in NameSpace,in Value,out Success)"}
   From_String_3_1_3_1_1_1_3_1_1 = RigVMDispatchNode(String=foot_r) {template="DISPATCH_RigDispatch_FromString(in String,out Result)"}
   > Set_Module_Metadata_2_1_1_1_1_1_1_1_1_2_1_1 = RigVMDispatchNode(Name=foot_r, NameSpace=Root, Value=(Type=Bone,Name="None"), Value.Name=@From_String_3_1_3_1_1_1_3_1_1.Result) {template="DISPATCH_RigDispatch_SetModuleMetadata(in Name,in NameSpace,in Value,out Success)"}
   From_String_3_1_3_1_2_1_4_1_1 = RigVMDispatchNode(String=hand_l) {template="DISPATCH_RigDispatch_FromString(in String,out Result)"}
   > Set_Module_Metadata_2_1_1_1_1_1_1_2_1_4_1_1 = RigVMDispatchNode(Name=hand_l, NameSpace=Root, Value=(Type=Bone,Name="None"), Value.Name=@From_String_3_1_3_1_2_1_4_1_1.Result) {template="DISPATCH_RigDispatch_SetModuleMetadata(in Name,in NameSpace,in Value,out Success)"}
   From_String_3_1_3_1_1_1_1_5_1_1 = RigVMDispatchNode(String=hand_r) {template="DISPATCH_RigDispatch_FromString(in String,out Result)"}
   > Set_Module_Metadata_2_1_1_1_1_1_1_1_1_1_4_1_1 = RigVMDispatchNode(Name=hand_r, NameSpace=Root, Value=(Type=Bone,Name="None"), Value.Name=@From_String_3_1_3_1_1_1_1_5_1_1.Result) {template="DISPATCH_RigDispatch_SetModuleMetadata(in Name,in NameSpace,in Value,out Success)"}
   From_String_3_1_3_1_2_1_3_2_1_1 = RigVMDispatchNode(String=lowerarm_l) {template="DISPATCH_RigDispatch_FromString(in String,out Result)"}
   > Set_Module_Metadata_2_1_1_1_1_1_1_2_1_3_1_1_1 = RigVMDispatchNode(Name=elbow_l, NameSpace=Root, Value=(Type=Bone,Name="None"), Value.Name=@From_String_3_1_3_1_2_1_3_2_1_1.Result) {template="DISPATCH_RigDispatch_SetModuleMetadata(in Name,in NameSpace,in Value,out Success)"}
   From_String_3_1_3_1_1_1_1_3_2_1_1 = RigVMDispatchNode(String=lowerarm_r) {template="DISPATCH_RigDispatch_FromString(in String,out Result)"}
   > Set_Module_Metadata_2_1_1_1_1_1_1_1_1_1_3_1_1_1 = RigVMDispatchNode(Name=elbow_r, NameSpace=Root, Value=(Type=Bone,Name="None"), Value.Name=@From_String_3_1_3_1_1_1_1_3_2_1_1.Result) {template="DISPATCH_RigDispatch_SetModuleMetadata(in Name,in NameSpace,in Value,out Success)"}
   From_String_3_1_1_1_1_2_1_2_1_1_1 = RigVMDispatchNode(String=ball_l) {template="DISPATCH_RigDispatch_FromString(in String,out Result)"}
   > Set_Module_Metadata_2_1_1_1_2_1_2_1_2_1_1_1 = RigVMDispatchNode(Name=toe_l, NameSpace=Root, Value=(Type=Bone,Name="None"), Value.Name=@From_String_3_1_1_1_1_2_1_2_1_1_1.Result) {template="DISPATCH_RigDispatch_SetModuleMetadata(in Name,in NameSpace,in Value,out Success)"}
   From_String_3_1_2_1_1_2_1_2_1_1_1 = RigVMDispatchNode(String=ball_r) {template="DISPATCH_RigDispatch_FromString(in String,out Result)"}
   > Set_Module_Metadata_2_1_1_1_1_2_1_2_1_2_1_1_1 = RigVMDispatchNode(Name=toe_r, NameSpace=Root, Value=(Type=Bone,Name="None"), Value.Name=@From_String_3_1_2_1_1_2_1_2_1_1_1.Result) {template="DISPATCH_RigDispatch_SetModuleMetadata(in Name,in NameSpace,in Value,out Success)"}
   n22 = &RigVMFunction_ControlFlowBranch(
    Condition=&RigVMFunction_MathBoolNot(Value=RigVMVariableNode() {variable=SkeletonFound}.Value).Result
    ExecuteContext=@RigVMFunction_Sequence_2.K
   )
   RigVMVariableNode(ExecuteContext=@n22.True) {variable=Skeleton}
   > Set_Module_Metadata_2_4 = RigVMDispatchNode(Name=Arm PV Position, NameSpace=Root, Value=3) {template="DISPATCH_RigDispatch_SetModuleMetadata(in Name,in NameSpace,in Value,out Success)"}
   > Set_Module_Metadata_2_3_1 = RigVMDispatchNode(Name=Leg PV Position, NameSpace=Root, Value=50) {template="DISPATCH_RigDispatch_SetModuleMetadata(in Name,in NameSpace,in Value,out Success)"}
   From_String_3_1_3_1_2_1_1 = RigVMDispatchNode(String=hand_l) {template="DISPATCH_RigDispatch_FromString(in String,out Result)"}
   > Set_Module_Metadata_2_1_1_1_1_1_1_2_1_1 = RigVMDispatchNode(Name=hand_l, NameSpace=Root, Value=(Type=Bone,Name="None"), Value.Name=@From_String_3_1_3_1_2_1_1.Result) {template="DISPATCH_RigDispatch_SetModuleMetadata(in Name,in NameSpace,in Value,out Success)"}
   From_String_3_1_3_1_1_1_1_1 = RigVMDispatchNode(String=hand_r) {template="DISPATCH_RigDispatch_FromString(in String,out Result)"}
   > Set_Module_Metadata_2_1_1_1_1_1_1_1_1_1_1 = RigVMDispatchNode(Name=hand_r, NameSpace=Root, Value=(Type=Bone,Name="None"), Value.Name=@From_String_3_1_3_1_1_1_1_1.Result) {template="DISPATCH_RigDispatch_SetModuleMetadata(in Name,in NameSpace,in Value,out Success)"}
   From_String_2_1_1_1_3 = RigVMDispatchNode(String=pelvis) {template="DISPATCH_RigDispatch_FromString(in String,out Result)"}
   > Set_Module_Metadata_3_1_1_3 = RigVMDispatchNode(Name=Pelvis Bone, NameSpace=Root, Value=(Type=Bone,Name="None"), Value.Name=@From_String_2_1_1_1_3.Result) {template="DISPATCH_RigDispatch_SetModuleMetadata(in Name,in NameSpace,in Value,out Success)"}
   From_String_1_1_1_1_3 = RigVMDispatchNode(String=spine_01) {template="DISPATCH_RigDispatch_FromString(in String,out Result)"}
   > Set_Module_Metadata_1_1_1_1_3 = RigVMDispatchNode(Name=Spine Start Bone, NameSpace=Root, Value=(Type=Bone,Name="None"), Value.Name=@From_String_1_1_1_1_3.Result) {template="DISPATCH_RigDispatch_SetModuleMetadata(in Name,in NameSpace,in Value,out Success)"}
   From_String_4_1_1_3 = RigVMDispatchNode(String=spine_05) {template="DISPATCH_RigDispatch_FromString(in String,out Result)"}
   > Set_Module_Metadata_2_2_1_1_1_3 = RigVMDispatchNode(Name=Spine End Bone, NameSpace=Root, Value=(Type=Bone,Name="None"), Value.Name=@From_String_4_1_1_3.Result) {template="DISPATCH_RigDispatch_SetModuleMetadata(in Name,in NameSpace,in Value,out Success)"}
   From_String_3_2_1_1_3 = RigVMDispatchNode(String=upperarm_l) {template="DISPATCH_RigDispatch_FromString(in String,out Result)"}
   > Set_Module_Metadata_2_1_2_1_1_3 = RigVMDispatchNode(Name=shoulder_l, NameSpace=Root, Value=(Type=Bone,Name="None"), Value.Name=@From_String_3_2_1_1_3.Result) {template="DISPATCH_RigDispatch_SetModuleMetadata(in Name,in NameSpace,in Value,out Success)"}
   From_String_3_1_4_1_1_3 = RigVMDispatchNode(String=upperarm_r) {template="DISPATCH_RigDispatch_FromString(in String,out Result)"}
   > Set_Module_Metadata_2_1_1_2_1_1_3 = RigVMDispatchNode(Name=shoulder_r, NameSpace=Root, Value=(Type=Bone,Name="None"), Value.Name=@From_String_3_1_4_1_1_3.Result) {template="DISPATCH_RigDispatch_SetModuleMetadata(in Name,in NameSpace,in Value,out Success)"}
   From_String_3_1_1_1_2_1_3 = RigVMDispatchNode(String=thigh_l) {template="DISPATCH_RigDispatch_FromString(in String,out Result)"}
   > Set_Module_Metadata_2_1_1_1_2_2_1_3 = RigVMDispatchNode(Name=thigh_l, NameSpace=Root, Value=(Type=Bone,Name="None"), Value.Name=@From_String_3_1_1_1_2_1_3.Result) {template="DISPATCH_RigDispatch_SetModuleMetadata(in Name,in NameSpace,in Value,out Success)"}
   From_String_3_1_2_1_2_1_3 = RigVMDispatchNode(String=thigh_r) {template="DISPATCH_RigDispatch_FromString(in String,out Result)"}
   > Set_Module_Metadata_2_1_1_1_1_2_2_1_3 = RigVMDispatchNode(Name=thigh_r, NameSpace=Root, Value=(Type=Bone,Name="None"), Value.Name=@From_String_3_1_2_1_2_1_3.Result) {template="DISPATCH_RigDispatch_SetModuleMetadata(in Name,in NameSpace,in Value,out Success)"}
   From_String_3_1_1_1_1_1_1_3 = RigVMDispatchNode(String=calf_l) {template="DISPATCH_RigDispatch_FromString(in String,out Result)"}
   > Set_Module_Metadata_2_1_1_1_2_1_1_1_3 = RigVMDispatchNode(Name=knee_l, NameSpace=Root, Value=(Type=Bone,Name="None"), Value.Name=@From_String_3_1_1_1_1_1_1_3.Result) {template="DISPATCH_RigDispatch_SetModuleMetadata(in Name,in NameSpace,in Value,out Success)"}
   From_String_3_1_2_1_1_1_1_3 = RigVMDispatchNode(String=calf_r) {template="DISPATCH_RigDispatch_FromString(in String,out Result)"}
   > Set_Module_Metadata_2_1_1_1_1_2_1_1_1_3 = RigVMDispatchNode(Name=knee_r, NameSpace=Root, Value=(Type=Bone,Name="None"), Value.Name=@From_String_3_1_2_1_1_1_1_3.Result) {template="DISPATCH_RigDispatch_SetModuleMetadata(in Name,in NameSpace,in Value,out Success)"}
   From_String_3_1_3_2_2_1_3 = RigVMDispatchNode(String=neck_01) {template="DISPATCH_RigDispatch_FromString(in String,out Result)"}
   > Set_Module_Metadata_2_1_1_1_1_1_2_2_1_3 = RigVMDispatchNode(Name=neck, NameSpace=Root, Value=(Type=Bone,Name="None"), Value.Name=@From_String_3_1_3_2_2_1_3.Result) {template="DISPATCH_RigDispatch_SetModuleMetadata(in Name,in NameSpace,in Value,out Success)"}
   From_String_3_1_3_2_1_2_1_3 = RigVMDispatchNode(String=Head) {template="DISPATCH_RigDispatch_FromString(in String,out Result)"}
   > Set_Module_Metadata_2_1_1_1_1_1_2_1_2_1_3 = RigVMDispatchNode(Name=head, NameSpace=Root, Value=(Type=Bone,Name="None"), Value.Name=@From_String_3_1_3_2_1_2_1_3.Result) {template="DISPATCH_RigDispatch_SetModuleMetadata(in Name,in NameSpace,in Value,out Success)"}
   From_String_3_1_3_1_2_2_1_3 = RigVMDispatchNode(String=foot_l) {template="DISPATCH_RigDispatch_FromString(in String,out Result)"}
   > Set_Module_Metadata_2_1_1_1_1_1_1_2_2_1_3 = RigVMDispatchNode(Name=foot_l, NameSpace=Root, Value=(Type=Bone,Name="None"), Value.Name=@From_String_3_1_3_1_2_2_1_3.Result) {template="DISPATCH_RigDispatch_SetModuleMetadata(in Name,in NameSpace,in Value,out Success)"}
   From_String_3_1_3_1_1_1_3_1_3 = RigVMDispatchNode(String=foot_r) {template="DISPATCH_RigDispatch_FromString(in String,out Result)"}
   > Set_Module_Metadata_2_1_1_1_1_1_1_1_1_2_1_3 = RigVMDispatchNode(Name=foot_r, NameSpace=Root, Value=(Type=Bone,Name="None"), Value.Name=@From_String_3_1_3_1_1_1_3_1_3.Result) {template="DISPATCH_RigDispatch_SetModuleMetadata(in Name,in NameSpace,in Value,out Success)"}
   From_String_3_1_3_1_2_1_4_1_3 = RigVMDispatchNode(String=hand_l) {template="DISPATCH_RigDispatch_FromString(in String,out Result)"}
   > Set_Module_Metadata_2_1_1_1_1_1_1_2_1_4_1_3 = RigVMDispatchNode(Name=hand_l, NameSpace=Root, Value=(Type=Bone,Name="None"), Value.Name=@From_String_3_1_3_1_2_1_4_1_3.Result) {template="DISPATCH_RigDispatch_SetModuleMetadata(in Name,in NameSpace,in Value,out Success)"}
   From_String_3_1_3_1_1_1_1_5_1_3 = RigVMDispatchNode(String=hand_r) {template="DISPATCH_RigDispatch_FromString(in String,out Result)"}
   > Set_Module_Metadata_2_1_1_1_1_1_1_1_1_1_4_1_3 = RigVMDispatchNode(Name=hand_r, NameSpace=Root, Value=(Type=Bone,Name="None"), Value.Name=@From_String_3_1_3_1_1_1_1_5_1_3.Result) {template="DISPATCH_RigDispatch_SetModuleMetadata(in Name,in NameSpace,in Value,out Success)"}
   From_String_3_1_3_1_2_1_3_2_1_3 = RigVMDispatchNode(String=lowerarm_l) {template="DISPATCH_RigDispatch_FromString(in String,out Result)"}
   > Set_Module_Metadata_2_1_1_1_1_1_1_2_1_3_1_1_3 = RigVMDispatchNode(Name=elbow_l, NameSpace=Root, Value=(Type=Bone,Name="None"), Value.Name=@From_String_3_1_3_1_2_1_3_2_1_3.Result) {template="DISPATCH_RigDispatch_SetModuleMetadata(in Name,in NameSpace,in Value,out Success)"}
   From_String_3_1_3_1_1_1_1_3_2_1_3 = RigVMDispatchNode(String=lowerarm_r) {template="DISPATCH_RigDispatch_FromString(in String,out Result)"}
   > Set_Module_Metadata_2_1_1_1_1_1_1_1_1_1_3_1_1_3 = RigVMDispatchNode(Name=elbow_r, NameSpace=Root, Value=(Type=Bone,Name="None"), Value.Name=@From_String_3_1_3_1_1_1_1_3_2_1_3.Result) {template="DISPATCH_RigDispatch_SetModuleMetadata(in Name,in NameSpace,in Value,out Success)"}
   From_String_3_1_1_1_1_2_1_2_1_1_2 = RigVMDispatchNode(String=ball_l) {template="DISPATCH_RigDispatch_FromString(in String,out Result)"}
   > Set_Module_Metadata_2_1_1_1_2_1_2_1_2_1_1_2 = RigVMDispatchNode(Name=toe_l, NameSpace=Root, Value=(Type=Bone,Name="None"), Value.Name=@From_String_3_1_1_1_1_2_1_2_1_1_2.Result) {template="DISPATCH_RigDispatch_SetModuleMetadata(in Name,in NameSpace,in Value,out Success)"}
   From_String_3_1_2_1_1_2_1_2_1_1_2 = RigVMDispatchNode(String=ball_r) {template="DISPATCH_RigDispatch_FromString(in String,out Result)"}
   > Set_Module_Metadata_2_1_1_1_1_2_1_2_1_2_1_1_2 = RigVMDispatchNode(Name=toe_r, NameSpace=Root, Value=(Type=Bone,Name="None"), Value.Name=@From_String_3_1_2_1_1_2_1_2_1_1_2.Result) {template="DISPATCH_RigDispatch_SetModuleMetadata(in Name,in NameSpace,in Value,out Success)"}
   From_String_3_1_2_1_1_2_1_2_1_1_3 = RigVMDispatchNode(String=root) {template="DISPATCH_RigDispatch_FromString(in String,out Result)"}
   > Set_Module_Metadata_2_1_1_1_1_2_1_2_1_2_1_1_3 = RigVMDispatchNode(Name=Root Bone, NameSpace=Root, Value=(Type=Bone,Name="None"), Value.Name=@From_String_3_1_2_1_1_2_1_2_1_1_3.Result) {template="DISPATCH_RigDispatch_SetModuleMetadata(in Name,in NameSpace,in Value,out Success)"}
   Make_Array = RigVMDispatchNode(Values=("spine","Spine1")) {template="DISPATCH_RigVMDispatch_ArrayMake(in Values,out Array)"}
   Make_Array_1_1_1_1_1_1_1_1_1_1_2 = RigVMDispatchNode(Values=("spine3")) {template="DISPATCH_RigVMDispatch_ArrayMake(in Values,out Array)"}
   n23 = RigVMDispatchNode(Index=0, Array=@Make_Array_1_1_1_1_1_1_1_1_1_1_2.Array) {template="DISPATCH_RigVMDispatch_ArrayGetAtIndex(in Array,in Index,out Element)"}
   RigVMFunction_MathBoolAnd = RigVMAggregateNode(
    A=&RigUnit_ItemExists(
     Item=(Type=Bone,Name="None")
     Item.Name=RigVMDispatchNode(Index=0, Array=@Make_Array.Array) {template="DISPATCH_RigVMDispatch_ArrayGetAtIndex(in Array,in Index,out Element)"}.Element
    ).Exists
    B=&RigUnit_ItemExists(
     Item=(Type=Bone,Name="None")
     Item.Name=RigVMDispatchNode(Index=1, Array=@Make_Array.Array) {template="DISPATCH_RigVMDispatch_ArrayGetAtIndex(in Array,in Index,out Element)"}.Element
    ).Exists
    D=&RigVMFunction_MathBoolNot(Value=RigVMVariableNode() {variable=SkeletonFound}.Value).Result
    C=&RigVMFunction_MathBoolNot(Value=&RigUnit_ItemExists(Item=(Type=Bone,Name="None"), Item.Name=@n23.Element).Exists).Result
   )
   n24 = &RigVMFunction_ControlFlowBranch(Condition=@RigVMFunction_MathBoolAnd.Result, ExecuteContext=@RigVMFunction_Sequence_2.B)
   RigVMVariableNode(Value=NewEnumerator2, ExecuteContext=@n24.True) {variable=Skeleton}
   > RigVMVariableNode(Value=true) {variable=SkeletonFound}
   From_String_2_1_1 = RigVMDispatchNode(String=Hips) {template="DISPATCH_RigDispatch_FromString(in String,out Result)"}
   > Set_Module_Metadata_3_1 = RigVMDispatchNode(Name=Pelvis Bone, NameSpace=Root, Value=(Type=Bone,Name="None"), Value.Name=@From_String_2_1_1.Result) {template="DISPATCH_RigDispatch_SetModuleMetadata(in Name,in NameSpace,in Value,out Success)"}
   From_String_1_1_1 = RigVMDispatchNode(String=Spine) {template="DISPATCH_RigDispatch_FromString(in String,out Result)"}
   > Set_Module_Metadata_1_1_1 = RigVMDispatchNode(Name=Spine Start Bone, NameSpace=Root, Value=(Type=Bone,Name="None"), Value.Name=@From_String_1_1_1.Result) {template="DISPATCH_RigDispatch_SetModuleMetadata(in Name,in NameSpace,in Value,out Success)"}
   From_String_4_1 = RigVMDispatchNode(String=Spine2) {template="DISPATCH_RigDispatch_FromString(in String,out Result)"}
   > Set_Module_Metadata_2_2_1_1 = RigVMDispatchNode(Name=Spine End Bone, NameSpace=Root, Value=(Type=Bone,Name="None"), Value.Name=@From_String_4_1.Result) {template="DISPATCH_RigDispatch_SetModuleMetadata(in Name,in NameSpace,in Value,out Success)"}
   From_String_3_2_1 = RigVMDispatchNode(String=LeftArm) {template="DISPATCH_RigDispatch_FromString(in String,out Result)"}
   > Set_Module_Metadata_2_1_2_1 = RigVMDispatchNode(Name=shoulder_l, NameSpace=Root, Value=(Type=Bone,Name="None"), Value.Name=@From_String_3_2_1.Result) {template="DISPATCH_RigDispatch_SetModuleMetadata(in Name,in NameSpace,in Value,out Success)"}
   From_String_3_1_4_1 = RigVMDispatchNode(String=RightArm) {template="DISPATCH_RigDispatch_FromString(in String,out Result)"}
   > Set_Module_Metadata_2_1_1_2_1 = RigVMDispatchNode(Name=shoulder_r, NameSpace=Root, Value=(Type=Bone,Name="None"), Value.Name=@From_String_3_1_4_1.Result) {template="DISPATCH_RigDispatch_SetModuleMetadata(in Name,in NameSpace,in Value,out Success)"}
   From_String_3_1_1_1_2 = RigVMDispatchNode(String=LeftUpLeg) {template="DISPATCH_RigDispatch_FromString(in String,out Result)"}
   > Set_Module_Metadata_2_1_1_1_2_2 = RigVMDispatchNode(Name=thigh_l, NameSpace=Root, Value=(Type=Bone,Name="None"), Value.Name=@From_String_3_1_1_1_2.Result) {template="DISPATCH_RigDispatch_SetModuleMetadata(in Name,in NameSpace,in Value,out Success)"}
   From_String_3_1_2_1_2 = RigVMDispatchNode(String=RightUpLeg) {template="DISPATCH_RigDispatch_FromString(in String,out Result)"}
   > Set_Module_Metadata_2_1_1_1_1_2_2 = RigVMDispatchNode(Name=thigh_r, NameSpace=Root, Value=(Type=Bone,Name="None"), Value.Name=@From_String_3_1_2_1_2.Result) {template="DISPATCH_RigDispatch_SetModuleMetadata(in Name,in NameSpace,in Value,out Success)"}
   From_String_3_1_1_1_1_1 = RigVMDispatchNode(String=LeftLeg) {template="DISPATCH_RigDispatch_FromString(in String,out Result)"}
   > Set_Module_Metadata_2_1_1_1_2_1_1 = RigVMDispatchNode(Name=knee_l, NameSpace=Root, Value=(Type=Bone,Name="None"), Value.Name=@From_String_3_1_1_1_1_1.Result) {template="DISPATCH_RigDispatch_SetModuleMetadata(in Name,in NameSpace,in Value,out Success)"}
   From_String_3_1_2_1_1_1 = RigVMDispatchNode(String=RightLeg) {template="DISPATCH_RigDispatch_FromString(in String,out Result)"}
   > Set_Module_Metadata_2_1_1_1_1_2_1_1 = RigVMDispatchNode(Name=knee_r, NameSpace=Root, Value=(Type=Bone,Name="None"), Value.Name=@From_String_3_1_2_1_1_1.Result) {template="DISPATCH_RigDispatch_SetModuleMetadata(in Name,in NameSpace,in Value,out Success)"}
   From_String_3_1_3_2_2 = RigVMDispatchNode(String=Neck) {template="DISPATCH_RigDispatch_FromString(in String,out Result)"}
   > Set_Module_Metadata_2_1_1_1_1_1_2_2 = RigVMDispatchNode(Name=neck, NameSpace=Root, Value=(Type=Bone,Name="None"), Value.Name=@From_String_3_1_3_2_2.Result) {template="DISPATCH_RigDispatch_SetModuleMetadata(in Name,in NameSpace,in Value,out Success)"}
   > Set_Module_Metadata_2_6_1_2 = RigVMDispatchNode(Name=Arm PV Position, NameSpace=Root, Value=10) {template="DISPATCH_RigDispatch_SetModuleMetadata(in Name,in NameSpace,in Value,out Success)"}
   > Set_Module_Metadata_2_3_3_1_1 = RigVMDispatchNode(Name=Leg PV Position, NameSpace=Root, Value=15) {template="DISPATCH_RigDispatch_SetModuleMetadata(in Name,in NameSpace,in Value,out Success)"}
   From_String_3_1_3_2_1_2 = RigVMDispatchNode(String=Head) {template="DISPATCH_RigDispatch_FromString(in String,out Result)"}
   > Set_Module_Metadata_2_1_1_1_1_1_2_1_2 = RigVMDispatchNode(Name=head, NameSpace=Root, Value=(Type=Bone,Name="None"), Value.Name=@From_String_3_1_3_2_1_2.Result) {template="DISPATCH_RigDispatch_SetModuleMetadata(in Name,in NameSpace,in Value,out Success)"}
   From_String_3_1_3_1_2_2 = RigVMDispatchNode(String=LeftFoot) {template="DISPATCH_RigDispatch_FromString(in String,out Result)"}
   > Set_Module_Metadata_2_1_1_1_1_1_1_2_2 = RigVMDispatchNode(Name=foot_l, NameSpace=Root, Value=(Type=Bone,Name="None"), Value.Name=@From_String_3_1_3_1_2_2.Result) {template="DISPATCH_RigDispatch_SetModuleMetadata(in Name,in NameSpace,in Value,out Success)"}
   From_String_3_1_3_1_1_1_3 = RigVMDispatchNode(String=RightFoot) {template="DISPATCH_RigDispatch_FromString(in String,out Result)"}
   > Set_Module_Metadata_2_1_1_1_1_1_1_1_1_2 = RigVMDispatchNode(Name=foot_r, NameSpace=Root, Value=(Type=Bone,Name="None"), Value.Name=@From_String_3_1_3_1_1_1_3.Result) {template="DISPATCH_RigDispatch_SetModuleMetadata(in Name,in NameSpace,in Value,out Success)"}
   From_String_3_1_3_1_2_1_4 = RigVMDispatchNode(String=LeftHand) {template="DISPATCH_RigDispatch_FromString(in String,out Result)"}
   > Set_Module_Metadata_2_1_1_1_1_1_1_2_1_4 = RigVMDispatchNode(Name=hand_l, NameSpace=Root, Value=(Type=Bone,Name="None"), Value.Name=@From_String_3_1_3_1_2_1_4.Result) {template="DISPATCH_RigDispatch_SetModuleMetadata(in Name,in NameSpace,in Value,out Success)"}
   From_String_3_1_3_1_1_1_1_5 = RigVMDispatchNode(String=RightHand) {template="DISPATCH_RigDispatch_FromString(in String,out Result)"}
   > Set_Module_Metadata_2_1_1_1_1_1_1_1_1_1_4 = RigVMDispatchNode(Name=hand_r, NameSpace=Root, Value=(Type=Bone,Name="None"), Value.Name=@From_String_3_1_3_1_1_1_1_5.Result) {template="DISPATCH_RigDispatch_SetModuleMetadata(in Name,in NameSpace,in Value,out Success)"}
   From_String_3_1_3_1_2_1_3_2 = RigVMDispatchNode(String=LeftForeArm) {template="DISPATCH_RigDispatch_FromString(in String,out Result)"}
   > Set_Module_Metadata_2_1_1_1_1_1_1_2_1_3_1 = RigVMDispatchNode(Name=elbow_l, NameSpace=Root, Value=(Type=Bone,Name="None"), Value.Name=@From_String_3_1_3_1_2_1_3_2.Result) {template="DISPATCH_RigDispatch_SetModuleMetadata(in Name,in NameSpace,in Value,out Success)"}
   From_String_3_1_3_1_1_1_1_3_2 = RigVMDispatchNode(String=RightForeARm) {template="DISPATCH_RigDispatch_FromString(in String,out Result)"}
   > Set_Module_Metadata_2_1_1_1_1_1_1_1_1_1_3_1 = RigVMDispatchNode(Name=elbow_r, NameSpace=Root, Value=(Type=Bone,Name="None"), Value.Name=@From_String_3_1_3_1_1_1_1_3_2.Result) {template="DISPATCH_RigDispatch_SetModuleMetadata(in Name,in NameSpace,in Value,out Success)"}
   > Set_Module_Metadata_6 = RigVMDispatchNode(Name=elbow_bone_offset, NameSpace=Self, Value=(X=0,Y=1,Z=-1)) {template="DISPATCH_RigDispatch_SetModuleMetadata(in Name,in NameSpace,in Value,out Success)"}
   From_String_3_1_1_1_1_2_1_2_1 = RigVMDispatchNode(String=LeftToeBase) {template="DISPATCH_RigDispatch_FromString(in String,out Result)"}
   > Set_Module_Metadata_2_1_1_1_2_1_2_1_2_1 = RigVMDispatchNode(Name=toe_l, NameSpace=Root, Value=(Type=Bone,Name="None"), Value.Name=@From_String_3_1_1_1_1_2_1_2_1.Result) {template="DISPATCH_RigDispatch_SetModuleMetadata(in Name,in NameSpace,in Value,out Success)"}
   From_String_3_1_2_1_1_2_1_2_1 = RigVMDispatchNode(String=RightToeBase) {template="DISPATCH_RigDispatch_FromString(in String,out Result)"}
   > Set_Module_Metadata_2_1_1_1_1_2_1_2_1_2_1 = RigVMDispatchNode(Name=toe_r, NameSpace=Root, Value=(Type=Bone,Name="None"), Value.Name=@From_String_3_1_2_1_1_2_1_2_1.Result) {template="DISPATCH_RigDispatch_SetModuleMetadata(in Name,in NameSpace,in Value,out Success)"}
   > Set_Module_Metadata_2_3_2_1_4_1_2 = RigVMDispatchNode(Name=Mirror Axis, NameSpace=Root, Value=(X=1,Y=0,Z=0)) {template="DISPATCH_RigDispatch_SetModuleMetadata(in Name,in NameSpace,in Value,out Success)"}
   RigVMVariableNode(Value=false, ExecuteContext=@RigVMFunction_Sequence_2.A) {variable=SkeletonFound}
   RigVMFunction_MathBoolNot = &RigVMFunction_MathBoolNot(Value=RigVMVariableNode() {variable=SkeletonFound}.Value)
   RigVMFunction_MathBoolOr = &RigVMFunction_MathBoolOr(B=@RigVMFunction_MathBoolNot.Result, A=RigVMVariableNode() {variable=Use Override Skeleton}.Value)
   n25 = &RigVMFunction_ControlFlowBranch(Condition=@RigVMFunction_MathBoolOr.Result, ExecuteContext=@RigVMFunction_Sequence.B)
   RigVMVariableNode(ExecuteContext=@n25.True, Value=RigVMVariableNode() {variable=Override Skeleton}.Value) {variable=Skeleton}
   RigVMFunctionReturnNode(ExecuteContext=@RigVMFunction_Sequence.C)
  ]}
 New Function_ContainedGraph {nodes [
   n1 = RigVMFunctionEntryNode()
   Less = &RigVMFunction_MathDoubleLess(B=0, A=@n1.Value.X) {template="Less::Execute(in A,in B,out Result)"}
   n2 = &RigVMFunction_MathDoubleLess(B=0, A=@n1.Value.Y) {template="Less::Execute(in A,in B,out Result)"}
   n3 = &RigVMFunction_MathDoubleLess(B=0, A=@n1.Value.Z) {template="Less::Execute(in A,in B,out Result)"}
   RigVMFunction_MathBoolOr_3 = RigVMAggregateNode(A=@Less.Result, B=@n2.Result, C=@n3.Result)
   Greater = &RigVMFunction_MathDoubleGreater(B=0, A=@n1.Value.X) {template="Greater::Execute(in A,in B,out Result)"}
   RigVMFunction_MathBoolOr = &RigVMFunction_MathBoolOr(A=@Greater.Result, B=@Less.Result)
   > RigVMFunctionReturnNode(
    Is Negative=RigVMDispatchNode(False=false, True=true, Condition=@RigVMFunction_MathBoolOr_3.Result) {template="DISPATCH_RigVMDispatch_If(in Condition,in True,in False,out Result)"}.Result
    X=@RigVMFunction_MathBoolOr.Result
    Y=&RigVMFunction_MathBoolOr(A=&RigVMFunction_MathDoubleGreater(B=0, A=@n1.Value.Y) {template="Greater::Execute(in A,in B,out Result)"}.Result, B=@n2.Result).Result
    Z=&RigVMFunction_MathBoolOr(A=&RigVMFunction_MathDoubleGreater(B=0, A=@n1.Value.Z) {template="Greater::Execute(in A,in B,out Result)"}.Result, B=@n3.Result).Result
   )
  ]}
 New Function_ContainedGraph {nodes [
   RigVMFunctionEntryNode()
   > RigVMFunction_Sequence = RigVMAggregateNode()
   RigVMVariableNode(
    Value=&RigUnit_CollectionChildrenArray(
     TypeToSearch=Bone
     bDefaultChildren=True
     bIncludeParent=true
     bRecursive=true
     Parent=&RigUnit_ResolveConnector(Connector=(Type=Connector,Name="Root"), SkipSocket=False).Result
    ).Items
    ExecuteContext=@RigVMFunction_Sequence.A
   ) {variable=Bone Tree}
   > For_Each = RigVMDispatchNode(Array=RigVMVariableNode() {variable=Bone Tree}.Value) {template="DISPATCH_RigVMDispatch_ArrayIterator(in Array,out Element,out Index,out Count,out Ratio)"}
   > RigVMDispatchNode(Array=RigVMVariableNode() {variable=Lengths}.Value) {template="DISPATCH_RigVMDispatch_ArrayReset(io Array)"}
   > RigVMVariableNode() {variable=Depth}
   > Set_Metadata = RigVMDispatchNode(
    Name=Parent Bone
    NameSpace=Self
    Value=/Script/ControlRig.RigUnit_HierarchyGetParent(bDefaultParent=True, Child=@For_Each.Element) {is=struct RigVMUnitNode}.Parent
    Item=@For_Each.Element
   ) {template="DISPATCH_RigDispatch_SetMetadata(in Item,in Name,in NameSpace,in Value,out Success)"}
   > Set_Metadata_1 = RigVMDispatchNode(
    Name=Child Bones
    NameSpace=Self
    Value=&RigUnit_CollectionChildrenArray(TypeToSearch=Bone, bDefaultChildren=True, bIncludeParent=False, bRecursive=False, Parent=@For_Each.Element).Items
    Item=@For_Each.Element
   ) {template="DISPATCH_RigDispatch_SetMetadata(in Item,in Name,in NameSpace,in Value,out Success)"}
   > Set_Metadata_2 = RigVMDispatchNode(
    Name=World Position
    NameSpace=Self
    Value=&RigUnit_GetTransform(bInitial=False, Item=@For_Each.Element).Transform
    Item=@For_Each.Element
   ) {template="DISPATCH_RigDispatch_SetMetadata(in Item,in Name,in NameSpace,in Value,out Success)"}
   n1 = &RigVMFunction_Sequence(ExecuteContext=@Set_Metadata_2.ExecuteContext)
   For_Each_1 = RigVMDispatchNode(
    Array=/Script/ControlRig.RigUnit_GetTransformItemArray(
     Space=GlobalSpace
     bInitial=False
     Items=&RigUnit_CollectionChildrenArray(TypeToSearch=Bone, bDefaultChildren=True, bIncludeParent=False, bRecursive=False, Parent=@For_Each.Element).Items
    ) {is=struct RigVMUnitNode}.Transforms
    ExecuteContext=@n1.A
   ) {template="DISPATCH_RigVMDispatch_ArrayIterator(in Array,out Element,out Index,out Count,out Ratio)"}
   RigVMFunction_MathVectorDistance = &RigVMFunction_MathVectorDistance(A=@For_Each_1.Element.Translation, B=&RigUnit_GetTransform(bInitial=False, Item=@For_Each.Element).Transform.Translation)
   > Add = RigVMDispatchNode(Element=@RigVMFunction_MathVectorDistance.Result, Array=RigVMVariableNode() {variable=Lengths}.Value) {template="DISPATCH_RigVMDispatch_ArrayAdd(io Array,in Element,out Index)"}
   Set_Metadata_3 = RigVMDispatchNode(Name=Lengths, NameSpace=Self, ExecuteContext=@n1.B, Value=RigVMVariableNode() {variable=Lengths}.Value, Item=@For_Each.Element) {template="DISPATCH_RigDispatch_SetMetadata(in Item,in Name,in NameSpace,in Value,out Success)"}
   > Set_Metadata_4 = RigVMDispatchNode(Name=Depth, NameSpace=Self, Value=RigVMVariableNode() {variable=Depth}.Value, Item=@For_Each.Element) {template="DISPATCH_RigDispatch_SetMetadata(in Item,in Name,in NameSpace,in Value,out Success)"}
   > RigVMVariableNode(Value=/Script/RigVM.RigVMFunction_MathIntAdd(B=1, A=RigVMVariableNode() {variable=Depth}.Value) {is=struct RigVMUnitNode, template="Add::Execute(in A,in B,out Result)"}.Result) {variable=Depth}
   ArraySum = /Script/RigVM.RigVMFunction_MathDoubleArraySum(Array=RigVMVariableNode() {variable=Lengths}.Value) {is=struct RigVMUnitNode, template="ArraySum::Execute(in Array,out Sum)"}
   IsNearlyZero = /Script/RigVM.RigVMFunction_MathDoubleIsNearlyZero(Tolerance=0, Value=@ArraySum.Sum) {is=struct RigVMUnitNode, template="IsNearlyZero::Execute(in Value,in Tolerance,out Result)"}
   > Set_Metadata_5 = RigVMDispatchNode(Name=isTip, NameSpace=Self, Value=@IsNearlyZero.Result, Item=@For_Each.Element) {template="DISPATCH_RigDispatch_SetMetadata(in Item,in Name,in NameSpace,in Value,out Success)"}
   For_Each_2 = RigVMDispatchNode(ExecuteContext=@RigVMFunction_Sequence.B, Array=RigVMVariableNode() {variable=Bone Tree}.Value) {template="DISPATCH_RigVMDispatch_ArrayIterator(in Array,out Element,out Index,out Count,out Ratio)"}
   n2 = &RigUnit_GetTransform(bInitial=False, Item=@For_Each_2.Element)
   RigVMFunction_MathTransformArrayToSRT = /Script/RigVM.RigVMFunction_MathTransformArrayToSRT(Transforms=&RigUnit_GetTransformItemArray(Space=GlobalSpace, bInitial=False, Items=RigVMVariableNode() {variable=Bone Tree}.Value).Transforms) {is=struct RigVMUnitNode}
   RigVMFunction_MathBoxFromArray = /Script/RigVM.RigVMFunction_MathBoxFromArray(Array=@RigVMFunction_MathTransformArrayToSRT.Translations) {is=struct RigVMUnitNode}
   n3 = &RigVMFunction_MathVectorDistance(A=(X=0,Y=0,Z=0), B=(X=0,Y=0,Z=0), B.Z=@n2.Transform.Translation.Z, A.Z=@RigVMFunction_MathBoxFromArray.Minimum.Z)
   LessEqual = /Script/RigVM.RigVMFunction_MathDoubleLessEqual(
    B=&RigVMFunction_MathVectorDistance(A=(X=0,Y=0,Z=0), B=(X=0,Y=0,Z=0), B.Z=@n2.Transform.Translation.Z, A.Z=@RigVMFunction_MathBoxFromArray.Maximum.Z).Result
    A=RigVMVariableNode() {variable=Current Dist}.Value
   ) {is=struct RigVMUnitNode, template="LessEqual::Execute(in A,in B,out Result)"}
   RigVMFunction_MathBoolAnd = &RigVMFunction_MathBoolAnd(
    A=/Script/RigVM.RigVMFunction_MathDoubleGreaterEqual(A=@n3.Result, B=RigVMVariableNode() {variable=Current Dist}.Value) {is=struct RigVMUnitNode, template="GreaterEqual::Execute(in A,in B,out Result)"}.Result
    B=@LessEqual.Result
   )
   n4 = &RigVMFunction_ControlFlowBranch(ExecuteContext=@For_Each_2.ExecuteContext, Condition=@RigVMFunction_MathBoolAnd.Result)
   RigVMVariableNode(ExecuteContext=@n4.True, Value=@n3.Result) {variable=Current Dist}
   > RigVMVariableNode(Value=@For_Each_2.Element) {variable=Current Pelvis}
   Greater = &RigVMFunction_MathDoubleGreater(B=@RigVMFunction_MathBoxFromArray.Size.X, A=@RigVMFunction_MathBoxFromArray.Size.Z) {template="Greater::Execute(in A,in B,out Result)"}
   n5 = &RigVMFunction_MathBoolAnd(
    A=@Greater.Result
    B=&RigVMFunction_MathDoubleGreater(B=@RigVMFunction_MathBoxFromArray.Size.Y, A=@RigVMFunction_MathBoxFromArray.Size.Z) {template="Greater::Execute(in A,in B,out Result)"}.Result
   )
   RigVMDispatchNode(ScreenColor=(R=1,G=1,B=1,A=1), ScreenDuration=0.05, Value="biped!!", ExecuteContext=@For_Each_2.Completed, Enabled=@n5.Result) {template="DISPATCH_RigVMDispatch_Print(in Prefix,in Value,in Enabled,in ScreenDuration,in ScreenColor)"}
   > RigVMVariableNode(Value=@n5.Result) {variable=Is Biped}
   RigVMFunctionReturnNode(
    ExecuteContext=@RigVMFunction_Sequence.C
    Pelvis=RigVMVariableNode() {variable=Current Pelvis}.Value
    isBiped=RigVMVariableNode() {variable=Is Biped}.Value
   )
  ]}
 New Function_ContainedGraph {nodes [
   RigVMFunctionEntryNode()
   > RigVMFunctionReturnNode()
  ]}
 New Function_ContainedGraph {nodes [
   n1 = RigVMFunctionEntryNode()
   RigVMFunction_Sequence = &RigVMFunction_Sequence(ExecuteContext=@n1.ExecuteContext)
   For_Each = RigVMDispatchNode(
    Array=&RigUnit_HierarchyGetParentsItemArray(
     bDefaultParent=True
     bIncludeChild=true
     bReverse=False
     Child=&RigUnit_ResolveConnector(Connector=(Type=Connector,Name="None"), SkipSocket=False).Result
    ).Parents
    ExecuteContext=@RigVMFunction_Sequence.A
   ) {template="DISPATCH_RigVMDispatch_ArrayIterator(in Array,out Element,out Index,out Count,out Ratio)"}
   Not_Equals = RigVMDispatchNode(B=(X=1,Y=1,Z=1), A=&RigUnit_GetTransform(bInitial=False, Item=@For_Each.Element).Transform.Scale3D) {template="DISPATCH_RigVMDispatch_CoreNotEquals(in A,in B,out Result)"}
   > RigVMVariableNode(Value=@Not_Equals.Result) {variable=HasScaleValues}
   RigVMFunction_ControlFlowBranch = &RigVMFunction_ControlFlowBranch(ExecuteContext=@For_Each.Completed, Condition=RigVMVariableNode() {variable=HasScaleValues}.Value)
   RigVMDispatchNode(
    Enabled=True
    ScreenColor=(R=1,G=1,B=1,A=1)
    ScreenDuration=2
    Value=Live scale values detected in skeleton. Please double check. Modules might fail.
    ExecuteContext=@RigVMFunction_ControlFlowBranch.True
   ) {template="DISPATCH_RigVMDispatch_Print(in Prefix,in Value,in Enabled,in ScreenDuration,in ScreenColor)"}
   RigVMFunctionReturnNode(ExecuteContext=@RigVMFunction_Sequence.B, Result=RigVMVariableNode() {variable=HasScaleValues}.Value)
  ]}
 RigVMFunctionLibrary {nodes [
   Axis to Bool = RigVMCollapseNode(Value=(X=0,Y=0,Z=0))
   Detect Scale = RigVMCollapseNode(ExecuteContext=())
   Find Largest Axis = RigVMCollapseNode(ExecuteContext=(), Vector=(X=0,Y=0,Z=0))
   Generate Biped Skeleton Data = RigVMCollapseNode(ExecuteContext=())
   INV Root = RigVMCollapseNode(ExecuteContext=())
   Is Right Side = RigVMCollapseNode(ExecuteContext=(), Item=(Type=Bone), Mirror Plane=(X=1,Y=1,Z=0))
   Map Skeleton Tree = RigVMCollapseNode(ExecuteContext=())
  ]}
 RigVMFunction_MathBoolAnd_1_1_1_1_1_1_1_1_1_ContainedGraph {nodes [
   n1 = RigVMFunctionEntryNode()
   n2 = &RigVMFunction_MathBoolAnd(B=@n1.C, A=&RigVMFunction_MathBoolAnd(A=@n1.A, B=@n1.B).Result)
   RigVMFunctionReturnNode(Result=&RigVMFunction_MathBoolAnd(B=@n1.E, A=&RigVMFunction_MathBoolAnd(B=@n1.D, A=@n2.Result).Result).Result)
  ]}
 RigVMFunction_MathBoolAnd_1_1_1_1_1_1_1_1_1_ContainedGraph {nodes [
   n1 = RigVMFunctionEntryNode()
   n2 = &RigVMFunction_MathBoolAnd(B=@n1.C, A=&RigVMFunction_MathBoolAnd(A=@n1.A, B=@n1.B).Result)
   RigVMFunctionReturnNode(Result=&RigVMFunction_MathBoolAnd(B=@n1.E, A=&RigVMFunction_MathBoolAnd(B=@n1.D, A=@n2.Result).Result).Result)
  ]}
 RigVMFunction_MathBoolAnd_1_1_1_1_1_2_ContainedGraph {nodes [
   n1 = RigVMFunctionEntryNode()
   n2 = &RigVMFunction_MathBoolAnd(A=@n1.A, B=@n1.B)
   RigVMFunctionReturnNode(Result=&RigVMFunction_MathBoolAnd(B=@n1.D, A=&RigVMFunction_MathBoolAnd(B=@n1.C, A=@n2.Result).Result).Result)
  ]}
 RigVMFunction_MathBoolAnd_1_1_1_2_ContainedGraph {nodes [
   n1 = RigVMFunctionEntryNode()
   RigVMFunctionReturnNode(Result=&RigVMFunction_MathBoolAnd(B=@n1.C, A=&RigVMFunction_MathBoolAnd(A=@n1.A, B=@n1.B).Result).Result)
  ]}
 RigVMFunction_MathBoolAnd_1_1_2_ContainedGraph {nodes [
   n1 = RigVMFunctionEntryNode()
   RigVMFunctionReturnNode(Result=&RigVMFunction_MathBoolAnd(B=@n1.C, A=&RigVMFunction_MathBoolAnd(A=@n1.A, B=@n1.B).Result).Result)
  ]}
 RigVMFunction_MathBoolAnd_1_2_ContainedGraph {nodes [
   n1 = RigVMFunctionEntryNode()
   RigVMFunctionReturnNode(Result=&RigVMFunction_MathBoolAnd(B=@n1.C, A=&RigVMFunction_MathBoolAnd(A=@n1.A, B=@n1.B).Result).Result)
  ]}
 RigVMFunction_MathBoolAnd_2_1_ContainedGraph {nodes [
   n1 = RigVMFunctionEntryNode()
   RigVMFunctionReturnNode(Result=&RigVMFunction_MathBoolAnd(B=@n1.C, A=&RigVMFunction_MathBoolAnd(A=@n1.A, B=@n1.B).Result).Result)
  ]}
 RigVMFunction_MathBoolAnd_3_ContainedGraph {nodes [
   n1 = RigVMFunctionEntryNode()
   RigVMFunction_MathBoolAnd = &RigVMFunction_MathBoolAnd(A=@n1.A, B=@n1.B)
   RigVMFunctionReturnNode(Result=&RigVMFunction_MathBoolAnd(B=@n1.D, A=&RigVMFunction_MathBoolAnd(B=@n1.C, A=@RigVMFunction_MathBoolAnd.Result).Result).Result)
  ]}
 RigVMFunction_MathBoolOr_3_1_ContainedGraph {nodes [
   n1 = RigVMFunctionEntryNode()
   RigVMFunctionReturnNode(Result=&RigVMFunction_MathBoolOr(B=@n1.C, A=&RigVMFunction_MathBoolOr(A=@n1.A, B=@n1.B).Result).Result)
  ]}
 RigVMFunction_Sequence_1_ContainedGraph {nodes [
   n1 = RigVMFunctionEntryNode()
   RigVMFunction_Sequence = &RigVMFunction_Sequence(ExecuteContext=@n1.ExecuteContext)
   n2 = &RigVMFunction_Sequence(ExecuteContext=@RigVMFunction_Sequence.B)
   RigVMFunctionReturnNode(A=@RigVMFunction_Sequence.A, B=@n2.A, C=@n2.B)
  ]}
 RigVMFunction_Sequence_1_ContainedGraph {nodes [
   n1 = RigVMFunctionEntryNode()
   RigVMFunction_Sequence = &RigVMFunction_Sequence(ExecuteContext=@n1.ExecuteContext)
   n2 = &RigVMFunction_Sequence(ExecuteContext=@RigVMFunction_Sequence.B)
   RigVMFunctionReturnNode(A=@RigVMFunction_Sequence.A, B=@n2.A, C=@n2.B)
  ]}
 RigVMFunction_Sequence_1_ContainedGraph {nodes [
   n1 = RigVMFunctionEntryNode()
   RigVMFunction_Sequence = &RigVMFunction_Sequence(ExecuteContext=@n1.ExecuteContext)
   n2 = &RigVMFunction_Sequence(ExecuteContext=@RigVMFunction_Sequence.B)
   RigVMFunctionReturnNode(A=@RigVMFunction_Sequence.A, B=@n2.A, C=@n2.B)
  ]}
 RigVMFunction_Sequence_2_1_ContainedGraph {nodes [
   n1 = RigVMFunctionEntryNode()
   n2 = &RigVMFunction_Sequence(ExecuteContext=@n1.ExecuteContext)
   n3 = &RigVMFunction_Sequence(ExecuteContext=@n2.B)
   n4 = &RigVMFunction_Sequence(ExecuteContext=@n3.B)
   n5 = &RigVMFunction_Sequence(ExecuteContext=@n4.B)
   n6 = &RigVMFunction_Sequence(ExecuteContext=@n5.B)
   n7 = &RigVMFunction_Sequence(ExecuteContext=@n6.B)
   n8 = &RigVMFunction_Sequence(ExecuteContext=@n7.B)
   n9 = &RigVMFunction_Sequence(ExecuteContext=@n8.B)
   n10 = &RigVMFunction_Sequence(ExecuteContext=@n9.B)
   n11 = &RigVMFunction_Sequence(ExecuteContext=@n10.B)
   RigVMFunctionReturnNode(A=@n2.A, B=@n3.A, C=@n4.A, D=@n5.A, E=@n6.A, F=@n7.A, G=@n8.A, H=@n9.A, I=@n10.A, J=@n11.A, K=@n11.B)
  ]}
 RigVMModel {nodes [
   RigUnit_BeginExecution = &RigUnit_BeginExecution()
   Connect Bone to Root Controls = RigVMCollapseNode(ExecuteContext=@RigUnit_BeginExecution.ExecutePin)
  ]}
 RigVMModel Backwards Solve Graph {nodes [
   n1 = /Script/ControlRig.RigUnit_InverseExecution() {is=struct RigVMUnitNode}
   INV Root = RigVMFunctionReferenceNode(ExecuteContext=@n1.ExecutePin)
  ]}
 RigVMModel Construction Event Graph {nodes [
   Map Skeleton Tree = RigVMFunctionReferenceNode(ExecuteContext=())
   > Debug Item = RigVMFunctionReferenceNode(Active=true, Scale=15, Thickness=2, Item=@"Map Skeleton Tree".Pelvis)
   n1 = &RigUnit_PrepareForExecution()
   Generate Biped Skeleton Data = RigVMFunctionReferenceNode(ExecuteContext=@n1.ExecutePin)
   > Create Controls = RigVMCollapseNode()
   Generate Metadata for Rig = RigVMCollapseNode(Value=@"Create Controls".ExecuteContext)
   Create Sockets = RigVMCollapseNode(ExecuteContext=@"Generate Metadata for Rig".Value)
   > Detect Scale = RigVMFunctionReferenceNode()
   > Scale and Colorize Controls = RigVMCollapseNode()
  ]}
]
hierarchy [
 Root {type=Connector, loc=<0 0 0>, rot=<0 0 0 1>}
 Root {type=Socket, loc=<0 0 0>, rot=<0 0 0 1>}
 MoveData_Speed {type=Curve, loc=<0 0 0>, rot=<0 0 0 1>}
 base_pose {type=Curve, loc=<0 0 0>, rot=<0 0 0 1>}
 calf_l_back_50 {type=Curve, loc=<0 0 0>, rot=<0 0 0 1>}
 calf_l_back_120 {type=Curve, loc=<0 0 0>, rot=<0 0 0 1>}
 calf_l_back_90 {type=Curve, loc=<0 0 0>, rot=<0 0 0 1>}
 calf_l_back_150 {type=Curve, loc=<0 0 0>, rot=<0 0 0 1>}
 calf_r_back_50 {type=Curve, loc=<0 0 0>, rot=<0 0 0 1>}
 calf_r_back_120 {type=Curve, loc=<0 0 0>, rot=<0 0 0 1>}
 calf_r_back_90 {type=Curve, loc=<0 0 0>, rot=<0 0 0 1>}
 calf_r_back_150 {type=Curve, loc=<0 0 0>, rot=<0 0 0 1>}
 clavicle_l_fwd_30 {type=Curve, loc=<0 0 0>, rot=<0 0 0 1>}
 clavicle_l_down_20 {type=Curve, loc=<0 0 0>, rot=<0 0 0 1>}
 clavicle_l_back_30 {type=Curve, loc=<0 0 0>, rot=<0 0 0 1>}
 clavicle_l_up_40 {type=Curve, loc=<0 0 0>, rot=<0 0 0 1>}
 clavicle_r_fwd_30 {type=Curve, loc=<0 0 0>, rot=<0 0 0 1>}
 clavicle_r_down_20 {type=Curve, loc=<0 0 0>, rot=<0 0 0 1>}
 clavicle_r_back_30 {type=Curve, loc=<0 0 0>, rot=<0 0 0 1>}
 clavicle_r_up_40 {type=Curve, loc=<0 0 0>, rot=<0 0 0 1>}
 clavicle_r_back_310 {type=Curve, loc=<0 0 0>, rot=<0 0 0 1>}
 clavicle_r_up_401 {type=Curve, loc=<0 0 0>, rot=<0 0 0 1>}
 foot_l_down_60 {type=Curve, loc=<0 0 0>, rot=<0 0 0 1>}
 foot_l_up_35 {type=Curve, loc=<0 0 0>, rot=<0 0 0 1>}
 foot_r_up_35 {type=Curve, loc=<0 0 0>, rot=<0 0 0 1>}
 foot_r_down_60 {type=Curve, loc=<0 0 0>, rot=<0 0 0 1>}
 hand_l_down_90 {type=Curve, loc=<0 0 0>, rot=<0 0 0 1>}
 hand_l_up_20 {type=Curve, loc=<0 0 0>, rot=<0 0 0 1>}
 hand_l_up_90 {type=Curve, loc=<0 0 0>, rot=<0 0 0 1>}
 hand_r_up_90 {type=Curve, loc=<0 0 0>, rot=<0 0 0 1>}
 hand_r_down_90 {type=Curve, loc=<0 0 0>, rot=<0 0 0 1>}
 hand_r_up_20 {type=Curve, loc=<0 0 0>, rot=<0 0 0 1>}
 Pose_0 {type=Curve, loc=<0 0 0>, rot=<0 0 0 1>}
 Pose_1 {type=Curve, loc=<0 0 0>, rot=<0 0 0 1>}
 Pose_2 {type=Curve, loc=<0 0 0>, rot=<0 0 0 1>}
 Pose_3 {type=Curve, loc=<0 0 0>, rot=<0 0 0 1>}
 Pose_4 {type=Curve, loc=<0 0 0>, rot=<0 0 0 1>}
 Pose_5 {type=Curve, loc=<0 0 0>, rot=<0 0 0 1>}
 Pose_6 {type=Curve, loc=<0 0 0>, rot=<0 0 0 1>}
 Pose_7 {type=Curve, loc=<0 0 0>, rot=<0 0 0 1>}
 Pose_8 {type=Curve, loc=<0 0 0>, rot=<0 0 0 1>}
 lowerarm_l_in_110 {type=Curve, loc=<0 0 0>, rot=<0 0 0 1>}
 lowerarm_l_in_35 {type=Curve, loc=<0 0 0>, rot=<0 0 0 1>}
 lowerarm_l_out_35 {type=Curve, loc=<0 0 0>, rot=<0 0 0 1>}
 lowerarm_l_in_50 {type=Curve, loc=<0 0 0>, rot=<0 0 0 1>}
 lowerarm_l_in_75 {type=Curve, loc=<0 0 0>, rot=<0 0 0 1>}
 lowerarm_l_in_10 {type=Curve, loc=<0 0 0>, rot=<0 0 0 1>}
 lowerarm_l_out_10 {type=Curve, loc=<0 0 0>, rot=<0 0 0 1>}
 lowerarm_l_in_90 {type=Curve, loc=<0 0 0>, rot=<0 0 0 1>}
 lowerarm_r_in_10 {type=Curve, loc=<0 0 0>, rot=<0 0 0 1>}
 lowerarm_r_out_10 {type=Curve, loc=<0 0 0>, rot=<0 0 0 1>}
 lowerarm_r_out_35 {type=Curve, loc=<0 0 0>, rot=<0 0 0 1>}
 lowerarm_r_in_75 {type=Curve, loc=<0 0 0>, rot=<0 0 0 1>}
 lowerarm_r_in_50 {type=Curve, loc=<0 0 0>, rot=<0 0 0 1>}
 lowerarm_r_in_110 {type=Curve, loc=<0 0 0>, rot=<0 0 0 1>}
 lowerarm_r_in_35 {type=Curve, loc=<0 0 0>, rot=<0 0 0 1>}
 lowerarm_r_in_90 {type=Curve, loc=<0 0 0>, rot=<0 0 0 1>}
 thigh_l_bck_10 {type=Curve, loc=<0 0 0>, rot=<0 0 0 1>}
 thigh_l_fwd_10 {type=Curve, loc=<0 0 0>, rot=<0 0 0 1>}
 thigh_l_in_45_out_90 {type=Curve, loc=<0 0 0>, rot=<0 0 0 1>}
 thigh_l_fwd_90 {type=Curve, loc=<0 0 0>, rot=<0 0 0 1>}
 thigh_l_bck_90 {type=Curve, loc=<0 0 0>, rot=<0 0 0 1>}
 thigh_l_bck_50 {type=Curve, loc=<0 0 0>, rot=<0 0 0 1>}
 thigh_l_out_110 {type=Curve, loc=<0 0 0>, rot=<0 0 0 1>}
 thigh_l_out_10 {type=Curve, loc=<0 0 0>, rot=<0 0 0 1>}
 thigh_l_fwd_45 {type=Curve, loc=<0 0 0>, rot=<0 0 0 1>}
 thigh_l_in_50 {type=Curve, loc=<0 0 0>, rot=<0 0 0 1>}
 thigh_l_fwd_110 {type=Curve, loc=<0 0 0>, rot=<0 0 0 1>}
 thigh_l_out_55 {type=Curve, loc=<0 0 0>, rot=<0 0 0 1>}
 thigh_l_out_85 {type=Curve, loc=<0 0 0>, rot=<0 0 0 1>}
 thigh_r_fwd_10 {type=Curve, loc=<0 0 0>, rot=<0 0 0 1>}
 thigh_r_out_55 {type=Curve, loc=<0 0 0>, rot=<0 0 0 1>}
 thigh_r_fwd_90 {type=Curve, loc=<0 0 0>, rot=<0 0 0 1>}
 thigh_r_fwd_45 {type=Curve, loc=<0 0 0>, rot=<0 0 0 1>}
 thigh_r_bck_10 {type=Curve, loc=<0 0 0>, rot=<0 0 0 1>}
 thigh_r_out_85 {type=Curve, loc=<0 0 0>, rot=<0 0 0 1>}
 thigh_r_in_50 {type=Curve, loc=<0 0 0>, rot=<0 0 0 1>}
 thigh_r_out_10 {type=Curve, loc=<0 0 0>, rot=<0 0 0 1>}
 thigh_r_out_110 {type=Curve, loc=<0 0 0>, rot=<0 0 0 1>}
 thigh_r_bck_50 {type=Curve, loc=<0 0 0>, rot=<0 0 0 1>}
 thigh_r_bck_90 {type=Curve, loc=<0 0 0>, rot=<0 0 0 1>}
 thigh_r_in_45_out_90 {type=Curve, loc=<0 0 0>, rot=<0 0 0 1>}
 thigh_r_fwd_110 {type=Curve, loc=<0 0 0>, rot=<0 0 0 1>}
 upperarm_l_back_45 {type=Curve, loc=<0 0 0>, rot=<0 0 0 1>}
 upperarm_l_out_55 {type=Curve, loc=<0 0 0>, rot=<0 0 0 1>}
 upperarm_l_out110_twist_in_70 {type=Curve, loc=<0 0 0>, rot=<0 0 0 1>}
 upperarm_l_out_85 {type=Curve, loc=<0 0 0>, rot=<0 0 0 1>}
 upperarm_l_fwd_110 {type=Curve, loc=<0 0 0>, rot=<0 0 0 1>}
 upperarm_l_in_10 {type=Curve, loc=<0 0 0>, rot=<0 0 0 1>}
 upperarm_l_out110_twist_out_70 {type=Curve, loc=<0 0 0>, rot=<0 0 0 1>}
 upperarm_l_out_twist_a {type=Curve, loc=<0 0 0>, rot=<0 0 0 1>}
 upperarm_l_fwd_15 {type=Curve, loc=<0 0 0>, rot=<0 0 0 1>}
 upperarm_l_out_15 {type=Curve, loc=<0 0 0>, rot=<0 0 0 1>}
 upperarm_l_back_10 {type=Curve, loc=<0 0 0>, rot=<0 0 0 1>}
 upperarm_l_out110 {type=Curve, loc=<0 0 0>, rot=<0 0 0 1>}
 upperarm_l_fwd_twist_in_a {type=Curve, loc=<0 0 0>, rot=<0 0 0 1>}
 upperarm_l_fwd_90 {type=Curve, loc=<0 0 0>, rot=<0 0 0 1>}
 upperarm_r_out_15 {type=Curve, loc=<0 0 0>, rot=<0 0 0 1>}
 upperarm_r_fwd_15 {type=Curve, loc=<0 0 0>, rot=<0 0 0 1>}
 upperarm_r_out_twist_a {type=Curve, loc=<0 0 0>, rot=<0 0 0 1>}
 upperarm_r_out110_twist_in_70 {type=Curve, loc=<0 0 0>, rot=<0 0 0 1>}
 upperarm_r_fwd_90 {type=Curve, loc=<0 0 0>, rot=<0 0 0 1>}
 upperarm_r_out110 {type=Curve, loc=<0 0 0>, rot=<0 0 0 1>}
 upperarm_r_back_10 {type=Curve, loc=<0 0 0>, rot=<0 0 0 1>}
 upperarm_r_out_85 {type=Curve, loc=<0 0 0>, rot=<0 0 0 1>}
 upperarm_r_in_10 {type=Curve, loc=<0 0 0>, rot=<0 0 0 1>}
 upperarm_r_back_45 {type=Curve, loc=<0 0 0>, rot=<0 0 0 1>}
 upperarm_r_out110_twist_out_70 {type=Curve, loc=<0 0 0>, rot=<0 0 0 1>}
 upperarm_r_out_55 {type=Curve, loc=<0 0 0>, rot=<0 0 0 1>}
 upperarm_r_fwd_twist_in_a {type=Curve, loc=<0 0 0>, rot=<0 0 0 1>}
 upperarm_r_fwd_110 {type=Curve, loc=<0 0 0>, rot=<0 0 0 1>}
 Distance {type=Curve, loc=<0 0 0>, rot=<0 0 0 1>}
 RemainingTurnYaw {type=Curve, loc=<0 0 0>, rot=<0 0 0 1>}
 TurnYawWeight {type=Curve, loc=<0 0 0>, rot=<0 0 0 1>}
 GroundDistance {type=Curve, loc=<0 0 0>, rot=<0 0 0 1>}
 DisableLHandIK {type=Curve, loc=<0 0 0>, rot=<0 0 0 1>}
 ScaleDownWeaponR {type=Curve, loc=<0 0 0>, rot=<0 0 0 1>}
 DisableLegIK {type=Curve, loc=<0 0 0>, rot=<0 0 0 1>}
 blendParent1 {type=Curve, loc=<0 0 0>, rot=<0 0 0 1>}
 DisableRHandIK {type=Curve, loc=<0 0 0>, rot=<0 0 0 1>}
 DisableLeftHandPoseOverride {type=Curve, loc=<0 0 0>, rot=<0 0 0 1>}
 applyHipfireOverridePose {type=Curve, loc=<0 0 0>, rot=<0 0 0 1>}
 DisableHandIKRetargeting {type=Curve, loc=<0 0 0>, rot=<0 0 0 1>}
 lowerarm_r_in_80 {type=Curve, loc=<0 0 0>, rot=<0 0 0 1>}
 lowerarm_l_in_80 {type=Curve, loc=<0 0 0>, rot=<0 0 0 1>}
 calf_r_back_130 {type=Curve, loc=<0 0 0>, rot=<0 0 0 1>}
 calf_l_back_130 {type=Curve, loc=<0 0 0>, rot=<0 0 0 1>}
 Pose_9 {type=Curve, loc=<0 0 0>, rot=<0 0 0 1>}
]
variables [
 Auto Find Skeleton {type=bool, default=False}
 Body Offset Control Name {type=FName, default=body_offset}
 Body Offset Control Settings {
  type=FRigUnit_HierarchyAddControlTransform_Settings
  default="(InitialSpace=LocalSpace,bUsePreferredRotationOrder=False,PreferredRotationOrder=YZX,Limits=(LimitTranslationX=(bMinimum=False,bMaximum=False),LimitTranslationY=(bMinimum=False,bMaximum=False),LimitTranslationZ=(bMinimum=False,bMaximum=False),LimitPitch=(b ...(truncated)"
 }
 Control Name Prefix {type=FName, default=None}
 Debug {type=bool, default=False}
 Generate Sockets {type=bool, default=True}
 Global Control Name {type=FName, default=global}
 Global Control Settings {
  type=FRigUnit_HierarchyAddControlTransform_Settings
  default="(InitialSpace=LocalSpace,bUsePreferredRotationOrder=False,PreferredRotationOrder=YZX,Limits=(LimitTranslationX=(bMinimum=False,bMaximum=False),LimitTranslationY=(bMinimum=False,bMaximum=False),LimitTranslationZ=(bMinimum=False,bMaximum=False),LimitPitch=(b ...(truncated)"
 }
 Orient to World {type=bool, default=True}
 Override Skeleton {type=EUserDefinedEnum_E_SkeletonData_9f63a266, default=NewEnumerator0}
 Root Control {type=FRigElementKey, default=(Type=None,Name="")}
 Root Control Name {type=FName, default=root}
 Root Control Scale {type=double, default=1}
 Root Ctrl Settings {
  type=FRigUnit_HierarchyAddControlTransform_Settings
  default="(InitialSpace=LocalSpace,bUsePreferredRotationOrder=False,PreferredRotationOrder=YZX,Limits=(LimitTranslationX=(bMinimum=False,bMaximum=False),LimitTranslationY=(bMinimum=False,bMaximum=False),LimitTranslationZ=(bMinimum=False,bMaximum=False),LimitPitch=(b ...(truncated)"
 }
 Root Module Settings {
  type=FUserDefinedStruct_#11
  default="(LeftColor_3_#2=(R=0,G=0.0231,B=0.4115,A=1),RightColor_5_#3=(R=1,G=0,B=0,A=1),CenterColor_9_#1=(R=1,G=1,B=0,A=1),ControlSize_12_#4=1,Pre ...(truncated)"
 }
 Skeleton {type=EUserDefinedEnum_E_SkeletonData_9f63a266, default=NewEnumerator0}
 Socket Built {type=bool, default=False}
 Spine Names {type="TArray<FName>", default=("spine")}
 Use Override Skeleton {type=bool, default=False}
 body offset control {type=FRigElementKey, default=(Type=None,Name="")}
 global control {type=FRigElementKey, default=(Type=None,Name="")}
]
